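# Idle glow effect for the CyberDelivery female character body, showing a pair of camera-offset sprite glows attached to the character.
# One Glow emitter bursts two persistent particles in separate spawn groups, each following a socket reader with mirrored offsets and socket-rotated facing and alignment.
# Two sprite renderers draw them, with separate RoundSize and RoundColor attributes alongside the main size and color.
# User.Variant selects both color sets, which are scaled by a time-of-day value.
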NiagaraSystem /BRCosmetics/Characters/Player/Female/Medium/Bodies/F_MED_CyberDelivery/FX/NS_Character_CyberDelivery_Idle
dependencies {
 usedDataInterfaces [&NiagaraDataInterfaceColorCurve, &NiagaraDataInterfaceSocketReader]
 usedDynamicInputs [
  &SimulationPosition
  &FNTimeOfDayScalar
  &FN_FrontEndFloatModifier
 ]
 usedModules [
  &self:SystemSpawnScript.SystemScriptSource.SystemScriptGraph.NiagaraNodeAssignment_3.SetVariables_#1
  &self:SystemSpawnScript.SystemScriptSource.SystemScriptGraph.NiagaraNodeAssignment_4.SetVariables_#2
  &SystemState
 ]
}
emitters {Glow {
  emitterUpdateScript [
   &EmitterState(
    Inactive Response=::NewEnumerator1
    Life Cycle Mode=::NewEnumerator1
    Loop Behavior=::NewEnumerator1
    Loop Duration Mode=::NewEnumerator1
    UseLoopDelay=false
   )
   &SpawnBurst_Instantaneous(Spawn Count=2, Use Spawn Probability=false, UseLoopCountLimit=false)
  ]
  particleSpawnScript [
   &FN_SetSpawnGroup(Number of Spawn Groups=2)
   &InitializeParticle(Sprite Size=<10 17>, Sprite Size Mode=::NewEnumerator4, UsePositionOffset=false, Write Material Random=true)
   &CameraOffset(Camera Offset Amount=1.6)
   /SetVariables(Particles.RoundSize:Vector2f=<8 15>) {isSetParameters=true}
  ]
  particleUpdateScript [
   &ParticleState(Kill Particles When Lifetime Has Elapsed=false, Let Infinitely Lived Particles Die When Emitter Deactivates=true)
   &SocketLocation(
    SamplePosOffset=true
    Sampled Position Offset=&MergeHelper_VectorSwitch(0=<13.5 0.3 7.3>, 1=<13.5 0.3 -7.3>, 2=?, 3=?, Index=&SpawnGroup(), Num=::NewEnumerator0)
    Socket Reader=$System.Socket Reader
    Socket Sampling Mode=::NewEnumerator2
    VelocitySpeedLimit=false
   )
   &SpriteFacingAndAlignment(
    Sprite Alignment=&RotateVectorByQuaternion(
     Vector To Rotate=&MergeHelper_VectorSwitch(0=<1 0.5 0.22>, 1=<1 0.5 -0.22>, 2=?, 3=?, Index=&SpawnGroup(), Num=::NewEnumerator0)
     Quaternion=$Output.SocketLocation.SampledRotation
    )
    Sprite Facing=&RotateVectorByQuaternion(
     Vector To Rotate=&MergeHelper_VectorSwitch(0=<-0.37 1 0>, 1=<-0.37 1 0>, 2=?, 3=?, Index=&SpawnGroup(), Num=::NewEnumerator0)
     Quaternion=$Output.SocketLocation.SampledRotation
    )
    WriteAlignment=true
    WriteFacing=true
   )
   &ScaleColor(
    Scale Mode=::NewEnumerator1
    Scale RGBA=/Niagara/DynamicInputs/Multiply/Multiply_Vector4_ByFloat(
     Vector4:Vector4f=&MakeVector4_FromLinearColor(LinearColor=&MergeHelper_ColorSwitch(
       0=<rgba 3.5 0.0911 0.2276 0.307>
       1=<rgba 3.5 0.0911 0.2276 0.307>
       2=<rgba 0.169 0.8293 4 0.263>
       3=<rgba 3.5 0.7624 0.1911 0.162>
       4=<rgba 0.0799 1.7427 2.4062 0.157>
       5=<rgba 0 1.5 0.6257 0.189>
       Index=&MakeIntFromFloat(Float=$User.Variant)
       Num=::NewEnumerator4
      ))
     Float:NiagaraFloat=$System.ToD
    )
    ScaleRGBA=true
   )
   /SetVariables(Particles.RoundColor:LinearColor=&Multiply_LinearColorByFloat(
     Linear Color=&MergeHelper_ColorSwitch(
      0=<rgba 2 0.0521 0.1301 0.124>
      1=<rgba 2 0.0521 0.1301 0.124>
      2=<rgba 0.0562 0.7447 4 0.199>
      3=<rgba 3.5 0.7624 0.1911 0.17>
      4=<rgba 0.0799 1.7427 2.4062 0.169>
      5=<rgba 0 2 2 0.12>
      Index=&MakeIntFromFloat(Float=$User.Variant)
      Num=::NewEnumerator4
     )
     Float=$System.ToD
    )) {isSetParameters=true}
  ]
  rendererClasses [&NiagaraSpriteRendererProperties]
  renderers [
   {rendererClass=&NiagaraSpriteRendererProperties}
   {rendererClass=&NiagaraSpriteRendererProperties, rendererIndex=1}
  ]
 }}
summary {
 systemName=NS_Character_CyberDelivery_Idle
 emitters [{emitterName=Glow, rendererClasses [&NiagaraSpriteRendererProperties]}]
 userVariables [
  User.bIsFrontEnd {type=NiagaraBool}
  User.bIsFrontEndPreview {type=NiagaraBool}
  User.Variant {type=NiagaraFloat, defaultValue=0, description=""}
 ]
}
