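# Material function for the CyberDelivery face accessory that overlays animated, emissive holographic facial expressions on incoming material attributes.
# Blueprint-driven index, running and timestamp parameters pick flipbook frames, reverting after an effect duration with timed UV glitches around transitions.
# The flipbook is recolored, square-masked and modulated by scanlines and a sine pulse, feeding emissive and base color.
# Emissive-texture, sphere and TechArt masks adjust roughness, specular and metallic.
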MaterialFunction /BRCosmetics/Characters/Player/Female/Medium/Bodies/F_MED_CyberDelivery/Materials/MF_CyberDelivery_FaceAcc
graph {expressions {
  MaterialExpressionAdd_0 {
   class=MaterialExpressionAdd
   inputs [
    A {from=MaterialExpressionMaterialFunctionCall_2, output=Result}
    B {from=""}
   ]
   properties {constB=0.5}
  }
  MaterialExpressionAdd_1 {
   class=MaterialExpressionAdd
   inputs [
    A {from=MaterialExpressionGetMaterialAttributes_0, output=Roughness}
    B {from=""}
   ]
   properties {constB=-0.175}
  }
  MaterialExpressionAdd_11 {
   class=MaterialExpressionAdd
   inputs [
    A {from=MaterialExpressionComponentMask_1}
    B {from=""}
   ]
   properties {constB=0.565}
  }
  MaterialExpressionAdd_12 {
   class=MaterialExpressionAdd
   inputs [
    A {from=MaterialExpressionComponentMask_1}
    B {from=""}
   ]
   properties {constB=0.85}
  }
  MaterialExpressionAdd_13 {
   class=MaterialExpressionAdd
   inputs [
    A {from=MaterialExpressionOneMinus_0}
    B {from=MaterialExpressionSaturate_10}
   ]
  }
  MaterialExpressionAdd_14 {
   class=MaterialExpressionAdd
   inputs [
    A {from=MaterialExpressionComponentMask_0}
    B {from=MaterialExpressionOneMinus_2}
   ]
  }
  MaterialExpressionAdd_2 {
   class=MaterialExpressionAdd
   inputs [
    A {from=MaterialExpressionGetMaterialAttributes_0, output=Specular}
    B {from=""}
   ]
   properties {constB=0.175}
  }
  MaterialExpressionAdd_3 {
   class=MaterialExpressionAdd
   inputs [
    A {from=MaterialExpressionMultiply_4}
    B {from=""}
   ]
   properties {constB=0.5}
  }
  MaterialExpressionAdd_4 {
   class=MaterialExpressionAdd
   inputs [
    A {from=MaterialExpressionMultiply_8}
    B {from=MaterialExpressionGetMaterialAttributes_0, output=BaseColor}
   ]
  }
  MaterialExpressionAdd_5 {
   class=MaterialExpressionAdd
   inputs [
    A {from=MaterialExpressionMaterialFunctionCall_11, output=UVs}
    B {from=MaterialExpressionIf_0}
   ]
   properties {constB=-0.5}
  }
  MaterialExpressionAdd_6 {
   class=MaterialExpressionAdd
   inputs [
    A {from=MaterialExpressionAdd_5}
    B {from=MaterialExpressionConstant2Vector_4}
   ]
  }
  MaterialExpressionAdd_8 {
   class=MaterialExpressionAdd
   inputs [
    A {from=MaterialExpressionSaturate_7}
    B {from=MaterialExpressionSphereMask_0}
   ]
  }
  MaterialExpressionAdd_9 {
   class=MaterialExpressionAdd
   inputs [
    A {from=MaterialExpressionGetMaterialAttributes_0, output=EmissiveColor}
    B {from=MaterialExpressionNamedRerouteUsage_0}
   ]
  }
  MaterialExpressionComponentMask_0 {class=MaterialExpressionComponentMask, inputs [None {from=MaterialExpressionFunctionInput_1}], properties {r=true}}
  MaterialExpressionComponentMask_1 {class=MaterialExpressionComponentMask, inputs [None {from=MaterialExpressionTextureCoordinate_0}], properties {g=true}}
  MaterialExpressionConstant2Vector_2 {class=MaterialExpressionConstant2Vector, properties {r=-1}}
  MaterialExpressionConstant2Vector_4 {class=MaterialExpressionConstant2Vector, properties {g=-0.275, r=-0.265}}
  MaterialExpressionConstant3Vector_0 {class=MaterialExpressionConstant3Vector, properties {constant=<rgba 1 1 1 0>}}
  MaterialExpressionConstant3Vector_1 {class=MaterialExpressionConstant3Vector, properties {constant=<rgba -0.25 9 152.5 0>}}
  MaterialExpressionConstant_0 {class=MaterialExpressionConstant, properties {r=1}}
  MaterialExpressionConstant_1 {class=MaterialExpressionConstant, properties {r=100}}
  MaterialExpressionConstant_12 {class=MaterialExpressionConstant, properties {r=0.43}}
  MaterialExpressionConstant_13 {class=MaterialExpressionConstant, properties {r=0.07}}
  MaterialExpressionConstant_14 {class=MaterialExpressionConstant, properties {r=0.11}}
  MaterialExpressionConstant_15 {class=MaterialExpressionConstant, properties {r=0.45}}
  MaterialExpressionConstant_16 {class=MaterialExpressionConstant, properties {r=50}}
  MaterialExpressionConstant_17 {class=MaterialExpressionConstant, properties {r=0.3}}
  MaterialExpressionConstant_18 {class=MaterialExpressionConstant}
  MaterialExpressionConstant_19 {class=MaterialExpressionConstant, properties {r=0.6}}
  MaterialExpressionConstant_2 {class=MaterialExpressionConstant, properties {r=200}}
  MaterialExpressionConstant_20 {class=MaterialExpressionConstant, properties {r=0.8}}
  MaterialExpressionConstant_21 {class=MaterialExpressionConstant, properties {r=0.4}}
  MaterialExpressionConstant_22 {class=MaterialExpressionConstant, properties {r=12.5664}}
  MaterialExpressionConstant_24 {class=MaterialExpressionConstant}
  MaterialExpressionConstant_25 {class=MaterialExpressionConstant, properties {r=1}}
  MaterialExpressionConstant_3 {class=MaterialExpressionConstant, properties {r=-0.15}}
  MaterialExpressionConstant_30 {class=MaterialExpressionConstant, properties {r=1}}
  MaterialExpressionConstant_4 {class=MaterialExpressionConstant, properties {r=2}}
  MaterialExpressionConstant_5 {class=MaterialExpressionConstant, properties {r=1}}
  MaterialExpressionConstant_6 {class=MaterialExpressionConstant, properties {r=-1}}
  MaterialExpressionConstant_7 {class=MaterialExpressionConstant, properties {r=1}}
  MaterialExpressionDesaturation_0 {
   class=MaterialExpressionDesaturation
   inputs [
    None {from=MaterialExpressionTextureSampleParameter2D_0, output=RGB}
    Fraction {from=MaterialExpressionScalarParameter_4, output=""}
   ]
   properties {}
  }
  MaterialExpressionDesaturation_1 {
   class=MaterialExpressionDesaturation
   inputs [
    None {from=MaterialExpressionTextureSampleParameter2D_0, output=RGB}
    Fraction {from=MaterialExpressionConstant_5, output=""}
   ]
   properties {}
  }
  MaterialExpressionDesaturation_2 {
   class=MaterialExpressionDesaturation
   inputs [
    None {from=MaterialExpressionMultiply_12, output=""}
    Fraction {from=MaterialExpressionConstant_7, output=""}
   ]
   properties {}
  }
  MaterialExpressionDivide_0 {
   class=MaterialExpressionDivide
   inputs [
    A {from=""}
    B {from=MaterialExpressionScalarParameter_3}
   ]
  }
  MaterialExpressionDivide_1 {
   class=MaterialExpressionDivide
   inputs [
    A {from=""}
    B {from=MaterialExpressionScalarParameter_6}
   ]
  }
  MaterialExpressionFloor_0 {class=MaterialExpressionFloor, inputs [None {from=MaterialExpressionIf_1, output=""}], properties {}}
  MaterialExpressionFloor_1 {class=MaterialExpressionFloor, inputs [None {from=MaterialExpressionAdd_11, output=""}], properties {}}
  MaterialExpressionFloor_3 {class=MaterialExpressionFloor, inputs [None {from=MaterialExpressionNamedRerouteUsage_1, output=""}], properties {}}
  MaterialExpressionFloor_4 {class=MaterialExpressionFloor, inputs [None {from=MaterialExpressionAdd_12, output=""}], properties {}}
  MaterialExpressionFunctionInput_0 {
   class=MaterialExpressionFunctionInput
   inputs [Preview {from="", output=""}]
   properties {inputName=MA, inputType=FunctionInput_MaterialAttributes}
  }
  MaterialExpressionFunctionInput_1 {
   class=MaterialExpressionFunctionInput
   inputs [Preview {from=MaterialExpressionConstant3Vector_0, output=""}]
   properties {bUsePreviewValueAsDefault=true, inputName=TechArtMasks, sortPriority=1}
  }
  MaterialExpressionFunctionOutput_1 {class=MaterialExpressionFunctionOutput, inputs [None {from=MaterialExpressionSetMaterialAttributes_0, output=""}]}
  MaterialExpressionGetMaterialAttributes_0 {
   class=MaterialExpressionGetMaterialAttributes
   inputs [None {from=MaterialExpressionFunctionInput_0, output=""}]
   properties {attributeGetTypes [
     #1
     #2
     #3
     #4
     #5
    ]}
  }
  MaterialExpressionIf_0 {
   class=MaterialExpressionIf
   inputs [
    A {from=MaterialExpressionStaticSwitchParameter_1, output=""}
    B {from="", output=""}
    "A > B" {from=MaterialExpressionConstant2Vector_2, output=""}
    "A == B" {from=MaterialExpressionConstant_24, output=""}
    "A < B" {from=MaterialExpressionConstant_24, output=""}
   ]
   properties {constB=1}
  }
  MaterialExpressionIf_1 {
   class=MaterialExpressionIf
   inputs [
    A {from=MaterialExpressionSaturate_0, output=""}
    B {from="", output=""}
    "A > B" {from=MaterialExpressionSaturate_0, output=""}
    "A == B" {from=MaterialExpressionConstant_25, output=""}
    "A < B" {from=MaterialExpressionConstant_25, output=""}
   ]
   properties {}
  }
  MaterialExpressionIf_2 {
   class=MaterialExpressionIf
   inputs [
    A {from=MaterialExpressionSaturate_6, output=""}
    B {from="", output=""}
    "A > B" {from=MaterialExpressionSaturate_6, output=""}
    "A == B" {from=MaterialExpressionConstant_30, output=""}
    "A < B" {from=MaterialExpressionConstant_30, output=""}
   ]
   properties {}
  }
  MaterialExpressionIf_4 {
   class=MaterialExpressionIf
   inputs [
    A {from=MaterialExpressionNamedRerouteUsage_2, output=""}
    B {from="", output=""}
    "A > B" {from=MaterialExpressionNamedRerouteUsage_2, output=""}
    "A == B" {from=MaterialExpressionConstant_0, output=""}
    "A < B" {from=MaterialExpressionConstant_0, output=""}
   ]
   properties {}
  }
  MaterialExpressionLinearInterpolate_0 {
   class=MaterialExpressionLinearInterpolate
   inputs [
    A {from=MaterialExpressionScalarParameter_1}
    B {from=""}
    Alpha {from=MaterialExpressionFloor_0}
   ]
   properties {constB=0}
  }
  MaterialExpressionLinearInterpolate_1 {
   class=MaterialExpressionLinearInterpolate
   inputs [
    A {from=MaterialExpressionMaterialFunctionCall_5, output=Glitched UV}
    B {from=MaterialExpressionAdd_6}
    Alpha {from=MaterialExpressionFloor_3}
   ]
  }
  MaterialExpressionLinearInterpolate_2 {
   class=MaterialExpressionLinearInterpolate
   inputs [
    A {from=""}
    B {from=""}
    Alpha {from=MaterialExpressionAdd_3}
   ]
   properties {constA=0.75}
  }
  MaterialExpressionLinearInterpolate_3 {
   class=MaterialExpressionLinearInterpolate
   inputs [
    A {from=MaterialExpressionSwitch_1}
    B {from=MaterialExpressionConstant_17}
    Alpha {from=MaterialExpressionScalarParameter_8}
   ]
  }
  MaterialExpressionLinearInterpolate_4 {
   class=MaterialExpressionLinearInterpolate
   inputs [
    A {from=MaterialExpressionVectorParameter_0, output=RGB}
    B {from=MaterialExpressionVectorParameter_1, output=RGB}
    Alpha {from=MaterialExpressionDesaturation_1}
   ]
  }
  MaterialExpressionLinearInterpolate_5 {
   class=MaterialExpressionLinearInterpolate
   inputs [
    A {from=""}
    B {from=MaterialExpressionConstant_21}
    Alpha {from=MaterialExpressionIf_4}
   ]
   properties {constA=0.8}
  }
  MaterialExpressionLinearInterpolate_6 {
   class=MaterialExpressionLinearInterpolate
   inputs [
    A {from=MaterialExpressionGetMaterialAttributes_0, output=Roughness}
    B {from=MaterialExpressionSaturate_2}
    Alpha {from=MaterialExpressionMultiply_14}
   ]
  }
  MaterialExpressionLinearInterpolate_7 {
   class=MaterialExpressionLinearInterpolate
   inputs [
    A {from=MaterialExpressionGetMaterialAttributes_0, output=Specular}
    B {from=MaterialExpressionSaturate_3}
    Alpha {from=MaterialExpressionMultiply_14}
   ]
  }
  MaterialExpressionLinearInterpolate_8 {
   class=MaterialExpressionLinearInterpolate
   inputs [
    A {from=MaterialExpressionGetMaterialAttributes_0, output=Metallic}
    B {from=MaterialExpressionMultiply_9}
    Alpha {from=MaterialExpressionMultiply_14}
   ]
  }
  MaterialExpressionLocalPosition_0 {class=MaterialExpressionLocalPosition, inputs [], properties {localOrigin=InstancePreSkinning}}
  MaterialExpressionMaterialFunctionCall_10 {
   class=MaterialExpressionMaterialFunctionCall
   inputs [
    Texture Coordinate {from=MaterialExpressionTextureCoordinate_1}
    X Pos Right {from=MaterialExpressionConstant_12}
    X Pos Left {from=MaterialExpressionConstant_13}
    Y Pos Up {from=MaterialExpressionConstant_14}
    Y Pos Down {from=MaterialExpressionConstant_15}
    SquareSoftness {from=MaterialExpressionConstant_16}
    RotateSquare {from=""}
    RotationAmount {from=""}
   ]
   properties {materialFunction=/Game/Accessories/FORT_Backpacks/Backpack_F_MED_GroovyReader_Play/Materials/MaterialFunctions/MF_SquareMaskGenerator}
  }
  MaterialExpressionMaterialFunctionCall_11 {
   class=MaterialExpressionMaterialFunctionCall
   inputs [
    "Animation  Phase (0-1)" {from=MaterialExpressionLinearInterpolate_3}
    Number of Rows {from=""}
    Number of Columns {from=""}
    Texture {from=""}
    UVs {from=MaterialExpressionRotator_1}
    MipBias/Level {from=""}
    "Use Mip Bias (T) Level (F)" {from=""}
    "Clamp Anim (see tooltip)" {from=""}
   ]
   properties {materialFunction=&FlipBook}
  }
  MaterialExpressionMaterialFunctionCall_12 {
   class=MaterialExpressionMaterialFunctionCall
   inputs [
    UVs {from=MaterialExpressionTextureCoordinate_2}
    Texture Scale {from=MaterialExpressionConstant_21}
   ]
   properties {materialFunction=&ScaleUVsByCenter}
  }
  MaterialExpressionMaterialFunctionCall_2 {
   class=MaterialExpressionMaterialFunctionCall
   inputs [
    ScanLine_Tiling_1 {from=MaterialExpressionConstant_1}
    Scanline_PannerSpeed_1 {from=MaterialExpressionConstant_3}
    ScanLine_Tiling {from=MaterialExpressionConstant_2}
    Scanline_PannerSpeed {from=MaterialExpressionConstant_6}
    Scanline_Rotation {from=""}
    UVs {from=""}
   ]
   properties {materialFunction=/Game/Global/Materials/MaterialGym/Materials/HolographicTexture/MF_HolographicTexture_ScanLines}
  }
  MaterialExpressionMaterialFunctionCall_5 {
   class=MaterialExpressionMaterialFunctionCall
   inputs [UV {from=MaterialExpressionAdd_6}]
   properties {materialFunction=&/MF_HolographicTextureGlitch_CyberDelivery}
  }
  MaterialExpressionMaterialFunctionCall_6 {
   class=MaterialExpressionMaterialFunctionCall
   inputs [
    Hue Shift Percentage {from=MaterialExpressionScalarParameter_5}
    Texture {from=MaterialExpressionDesaturation_0}
   ]
   properties {materialFunction=&HueShift}
  }
  MaterialExpressionMultiply_0 {
   class=MaterialExpressionMultiply
   inputs [
    A {from=MaterialExpressionMultiply_1}
    B {from=MaterialExpressionLinearInterpolate_2}
   ]
  }
  MaterialExpressionMultiply_1 {
   class=MaterialExpressionMultiply
   inputs [
    A {from=MaterialExpressionSaturate_4}
    B {from=MaterialExpressionMultiply_7}
   ]
  }
  MaterialExpressionMultiply_10 {
   class=MaterialExpressionMultiply
   inputs [
    A {from=MaterialExpressionDesaturation_2}
    B {from=""}
   ]
   properties {constB=1000}
  }
  MaterialExpressionMultiply_11 {
   class=MaterialExpressionMultiply
   inputs [
    A {from=MaterialExpressionSubtract_2}
    B {from=MaterialExpressionDivide_0}
   ]
  }
  MaterialExpressionMultiply_12 {
   class=MaterialExpressionMultiply
   inputs [
    A {from=MaterialExpressionTextureSample_0, output=RGB}
    B {from=MaterialExpressionOneMinus_1}
   ]
  }
  MaterialExpressionMultiply_14 {
   class=MaterialExpressionMultiply
   inputs [
    A {from=MaterialExpressionSaturate_11}
    B {from=MaterialExpressionSaturate_8}
   ]
  }
  MaterialExpressionMultiply_15 {
   class=MaterialExpressionMultiply
   inputs [
    A {from=MaterialExpressionDivide_1}
    B {from=MaterialExpressionSubtract_6}
   ]
  }
  MaterialExpressionMultiply_2 {
   class=MaterialExpressionMultiply
   inputs [
    A {from=MaterialExpressionScalarParameter_9}
    B {from=MaterialExpressionStaticSwitchParameter_0}
   ]
  }
  MaterialExpressionMultiply_3 {
   class=MaterialExpressionMultiply
   inputs [
    A {from=MaterialExpressionSubtract_4}
    B {from=MaterialExpressionDivide_1}
   ]
  }
  MaterialExpressionMultiply_4 {
   class=MaterialExpressionMultiply
   inputs [
    A {from=MaterialExpressionSine_0}
    B {from=""}
   ]
   properties {constB=0.5}
  }
  MaterialExpressionMultiply_5 {
   class=MaterialExpressionMultiply
   inputs [
    A {from=MaterialExpressionMultiply_2}
    B {from=MaterialExpressionTextureSampleParameter2D_0, output=A}
   ]
  }
  MaterialExpressionMultiply_6 {
   class=MaterialExpressionMultiply
   inputs [
    A {from=MaterialExpressionTime_1}
    B {from=""}
   ]
   properties {constB=4}
  }
  MaterialExpressionMultiply_7 {
   class=MaterialExpressionMultiply
   inputs [
    A {from=MaterialExpressionMultiply_5}
    B {from=MaterialExpressionMaterialFunctionCall_10, output=SquareMask}
   ]
   properties {constB=0.5}
  }
  MaterialExpressionMultiply_8 {
   class=MaterialExpressionMultiply
   inputs [
    A {from=MaterialExpressionPower_0}
    B {from=""}
   ]
   properties {constB=10}
  }
  MaterialExpressionMultiply_9 {
   class=MaterialExpressionMultiply
   inputs [
    A {from=MaterialExpressionMultiply_14}
    B {from=""}
   ]
   properties {constB=0}
  }
  MaterialExpressionNamedRerouteDeclaration_0 {
   class=MaterialExpressionNamedRerouteDeclaration
   inputs [None {from=MaterialExpressionMultiply_0}]
   properties {name=FacialExpressions, nodeColor=<rgba 0 0.5078 1 1>}
  }
  MaterialExpressionNamedRerouteDeclaration_1 {
   class=MaterialExpressionNamedRerouteDeclaration
   inputs [None {from=MaterialExpressionIf_2}]
   properties {name=GlitchTiming, nodeColor=<rgba 0.6641 0 1 1>}
  }
  MaterialExpressionNamedRerouteUsage_0 {class=MaterialExpressionNamedRerouteUsage}
  MaterialExpressionNamedRerouteUsage_1 {class=MaterialExpressionNamedRerouteUsage}
  MaterialExpressionNamedRerouteUsage_2 {class=MaterialExpressionNamedRerouteUsage}
  MaterialExpressionOneMinus_0 {class=MaterialExpressionOneMinus, inputs [None {from=MaterialExpressionSaturate_9}]}
  MaterialExpressionOneMinus_1 {class=MaterialExpressionOneMinus, inputs [None {from=MaterialExpressionAdd_13}]}
  MaterialExpressionOneMinus_2 {class=MaterialExpressionOneMinus, inputs [None {from=MaterialExpressionScalarParameter_7}]}
  MaterialExpressionPower_0 {
   class=MaterialExpressionPower
   inputs [
    Base {from=MaterialExpressionMultiply_1}
    Exp {from=""}
   ]
   properties {constExponent=8}
  }
  MaterialExpressionPower_1 {
   class=MaterialExpressionPower
   inputs [
    Base {from=MaterialExpressionLinearInterpolate_4}
    Exp {from=MaterialExpressionScalarParameter_0}
   ]
  }
  MaterialExpressionRotator_1 {
   class=MaterialExpressionRotator
   inputs [
    Coordinate {from=MaterialExpressionMaterialFunctionCall_12, output=UVs}
    Time {from=MaterialExpressionConstant_22, output=""}
   ]
   properties {}
  }
  MaterialExpressionSaturate_0 {class=MaterialExpressionSaturate, inputs [None {from=MaterialExpressionMultiply_11}]}
  MaterialExpressionSaturate_1 {class=MaterialExpressionSaturate, inputs [None {from=MaterialExpressionMultiply_15}]}
  MaterialExpressionSaturate_10 {class=MaterialExpressionSaturate, inputs [None {from=MaterialExpressionFloor_1}]}
  MaterialExpressionSaturate_11 {class=MaterialExpressionSaturate, inputs [None {from=MaterialExpressionAdd_14}]}
  MaterialExpressionSaturate_2 {class=MaterialExpressionSaturate, inputs [None {from=MaterialExpressionAdd_1}]}
  MaterialExpressionSaturate_3 {class=MaterialExpressionSaturate, inputs [None {from=MaterialExpressionAdd_2}]}
  MaterialExpressionSaturate_4 {class=MaterialExpressionSaturate, inputs [None {from=MaterialExpressionAdd_0}]}
  MaterialExpressionSaturate_5 {class=MaterialExpressionSaturate, inputs [None {from=MaterialExpressionMultiply_3}]}
  MaterialExpressionSaturate_6 {class=MaterialExpressionSaturate, inputs [None {from=MaterialExpressionSubtract_5}]}
  MaterialExpressionSaturate_7 {class=MaterialExpressionSaturate, inputs [None {from=MaterialExpressionMultiply_10}]}
  MaterialExpressionSaturate_8 {class=MaterialExpressionSaturate, inputs [None {from=MaterialExpressionAdd_8}]}
  MaterialExpressionSaturate_9 {class=MaterialExpressionSaturate, inputs [None {from=MaterialExpressionFloor_4}]}
  MaterialExpressionScalarParameter_0 {
   class=MaterialExpressionScalarParameter
   properties {defaultValue=1, group=CyberDelivery, parameterName=ExpressionsColor_Power, sortPriority=60}
  }
  MaterialExpressionScalarParameter_1 {
   class=MaterialExpressionScalarParameter
   properties {desc=Controlled by the blueprint as well, group=CyberDelivery, parameterName=BP_ExpressionIndex, sortPriority=100}
  }
  MaterialExpressionScalarParameter_2 {
   class=MaterialExpressionScalarParameter
   properties {defaultValue=9.999999843067494e+17, group=CyberDelivery, parameterName=BP_GetTime, sortPriority=50}
  }
  MaterialExpressionScalarParameter_3 {
   class=MaterialExpressionScalarParameter
   properties {defaultValue=6, group=CyberDelivery, parameterName=ReactiveFaces_EffectLength, sortPriority=50}
  }
  MaterialExpressionScalarParameter_4 {class=MaterialExpressionScalarParameter, properties {group=CyberDelivery, parameterName=Expressions_Desaturation, sortPriority=60}}
  MaterialExpressionScalarParameter_5 {class=MaterialExpressionScalarParameter, properties {group=CyberDelivery, parameterName=Expressions_HueShift, sortPriority=60}}
  MaterialExpressionScalarParameter_6 {
   class=MaterialExpressionScalarParameter
   properties {defaultValue=0.6, group=CyberDelivery, parameterName=ReactiveFaces_GlitchDuration, sortPriority=50}
  }
  MaterialExpressionScalarParameter_7 {class=MaterialExpressionScalarParameter, properties {defaultValue=0.5, group=CyberDelivery, parameterName=HoneyCombSpecular_Influence}}
  MaterialExpressionScalarParameter_8 {
   class=MaterialExpressionScalarParameter
   properties {desc=Controlled by the blueprint as well, group=CyberDelivery, parameterName=BP_IsRunning, sortPriority=100}
  }
  MaterialExpressionScalarParameter_9 {
   class=MaterialExpressionScalarParameter
   properties {defaultValue=1, group=CyberDelivery, parameterName=Expressions_Multiplier, sortPriority=60}
  }
  MaterialExpressionSetMaterialAttributes_0 {
   class=MaterialExpressionSetMaterialAttributes
   inputs [
    MaterialAttributes {from=MaterialExpressionGetMaterialAttributes_0, output=MaterialAttributes}
    BaseColor {from=MaterialExpressionAdd_4, output=""}
    EmissiveColor {from=MaterialExpressionAdd_9, output=""}
    Roughness {from=MaterialExpressionLinearInterpolate_6, output=""}
    Specular {from=MaterialExpressionLinearInterpolate_7, output=""}
    Metallic {from=MaterialExpressionLinearInterpolate_8, output=""}
   ]
   properties {attributeSetTypes [
     #1
     #2
     #3
     #4
     #5
    ]}
  }
  MaterialExpressionShaderStageSwitch_0 {
   class=MaterialExpressionShaderStageSwitch
   inputs [
    PixelShader {from=MaterialExpressionVertexInterpolator_0, output=PS}
    VertexShader {from=MaterialExpressionLocalPosition_0, output=XYZ}
   ]
   properties {}
  }
  MaterialExpressionSine_0 {class=MaterialExpressionSine, inputs [None {from=MaterialExpressionMultiply_6}]}
  MaterialExpressionSphereMask_0 {
   class=MaterialExpressionSphereMask
   inputs [
    A {from=MaterialExpressionShaderStageSwitch_0, output=""}
    B {from=MaterialExpressionConstant3Vector_1, output=""}
    Radius {from="", output=""}
    Hardness {from="", output=""}
   ]
   properties {attenuationRadius=5.65}
  }
  MaterialExpressionStaticSwitchParameter_0 {
   class=MaterialExpressionStaticSwitchParameter
   inputs [
    True {from=MaterialExpressionPower_1, output=""}
    False {from=MaterialExpressionMaterialFunctionCall_6, output=Result}
   ]
   properties {group=CyberDelivery, parameterName=OverrideExpressionsColors, sortPriority=60}
  }
  MaterialExpressionStaticSwitchParameter_1 {
   class=MaterialExpressionStaticSwitchParameter
   inputs [
    True {from=MaterialExpressionConstant_4, output=""}
    False {from=MaterialExpressionLinearInterpolate_0, output=""}
   ]
   properties {group=CyberDelivery, parameterName=DEBUG_ShowExclamation, sortPriority=-1}
  }
  MaterialExpressionSubtract_1 {
   class=MaterialExpressionSubtract
   inputs [
    A {from=MaterialExpressionTime_0}
    B {from=MaterialExpressionScalarParameter_2}
   ]
  }
  MaterialExpressionSubtract_2 {
   class=MaterialExpressionSubtract
   inputs [
    A {from=MaterialExpressionSubtract_1}
    B {from=""}
   ]
   properties {constB=0.25}
  }
  MaterialExpressionSubtract_4 {
   class=MaterialExpressionSubtract
   inputs [
    A {from=MaterialExpressionSubtract_1}
    B {from=MaterialExpressionScalarParameter_3}
   ]
   properties {constB=0.18}
  }
  MaterialExpressionSubtract_5 {
   class=MaterialExpressionSubtract
   inputs [
    A {from=MaterialExpressionSaturate_1}
    B {from=MaterialExpressionSaturate_5}
   ]
  }
  MaterialExpressionSubtract_6 {
   class=MaterialExpressionSubtract
   inputs [
    A {from=MaterialExpressionSubtract_1}
    B {from=""}
   ]
   properties {constB=0.18}
  }
  MaterialExpressionSwitch_1 {
   class=MaterialExpressionSwitch
   inputs [
    SwitchValue {from=MaterialExpressionStaticSwitchParameter_1, output=""}
    Default {from="", output=""}
    None {from=MaterialExpressionConstant_18, output=""}
    None {from=MaterialExpressionConstant_17, output=""}
    None {from=MaterialExpressionConstant_19, output=""}
    None {from=MaterialExpressionConstant_20, output=""}
   ]
   properties {inputs [
     {
      inputName=None
      input {
       expression=&self:MaterialExpressionConstant_18
       inputName=None
       mask=0
       maskA=0
       maskB=0
       maskG=0
       maskR=0
       outputIndex=0
      }
     }
     {
      inputName=None
      input {
       expression=&self:MaterialExpressionConstant_17
       inputName=None
       mask=0
       maskA=0
       maskB=0
       maskG=0
       maskR=0
       outputIndex=0
      }
     }
     {
      inputName=None
      input {
       expression=&self:MaterialExpressionConstant_19
       inputName=None
       mask=0
       maskA=0
       maskB=0
       maskG=0
       maskR=0
       outputIndex=0
      }
     }
     {
      inputName=None
      input {
       expression=&self:MaterialExpressionConstant_20
       inputName=None
       mask=0
       maskA=0
       maskB=0
       maskG=0
       maskR=0
       outputIndex=0
      }
     }
    ]}
  }
  MaterialExpressionTextureCoordinate_0 {class=MaterialExpressionTextureCoordinate}
  MaterialExpressionTextureCoordinate_1 {class=MaterialExpressionTextureCoordinate}
  MaterialExpressionTextureCoordinate_2 {class=MaterialExpressionTextureCoordinate}
  MaterialExpressionTextureSampleParameter2D_0 {
   class=MaterialExpressionTextureSampleParameter2D
   inputs [
    UVs {from=MaterialExpressionLinearInterpolate_1, output=""}
    Apply View MipBias {from="", output=""}
   ]
   properties {
    group=CyberDelivery
    parameterName=CyberDelivery_Flipbook
    samplerSource=SSM_Clamp_WorldGroupSettings
    texture=/BRCosmetics/Characters/Player/Female/Medium/Bodies/F_MED_CyberDelivery/Textures/T_CyberDelivery_Face_Flipbook
   }
  }
  MaterialExpressionTextureSample_0 {
   class=MaterialExpressionTextureSample
   inputs [
    UVs {from=""}
    Tex {from=""}
    Apply View MipBias {from=""}
   ]
   properties {
    samplerSource=SSM_Wrap_WorldGroupSettings
    texture=/BRCosmetics/Characters/Player/Female/Medium/Bodies/F_MED_CyberDelivery/Textures/T_CyberDelivery_FaceAcc_E
   }
  }
  MaterialExpressionTime_0 {class=MaterialExpressionTime}
  MaterialExpressionTime_1 {class=MaterialExpressionTime}
  MaterialExpressionVectorParameter_0 {
   class=MaterialExpressionVectorParameter
   properties {group=CyberDelivery, parameterName=ExpressionColor_01, sortPriority=51, defaultValue=<rgba 0.6684 0.0296 0 1>}
  }
  MaterialExpressionVectorParameter_1 {
   class=MaterialExpressionVectorParameter
   properties {group=CyberDelivery, parameterName=ExpressionColor_02, sortPriority=51, defaultValue=<rgba 1 0.482 0.0936 1>}
  }
  MaterialExpressionVertexInterpolator_0 {class=MaterialExpressionVertexInterpolator, inputs [VS {from=MaterialExpressionLocalPosition_0, output=XYZ}], properties {}}
 }}
