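# Animation Blueprint for the F_MED_CyberDelivery female cosmetic body that follows the parent mesh's pose and adjusts its belt and pistol bones.
# Copy Pose From Mesh feeds a Control Rig, then cached poses carry it through chained additive bone-space Modify Bone offsets to the output.
# The update event derives crouch-idle and crouched side-leg alphas from movement direction and thigh-to-pelvis rotation, though the AnimGraph shows no reads of these variables.
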AnimBlueprint /BRCosmetics/Characters/Player/Female/Medium/Bodies/F_MED_CyberDelivery/Meshes/F_MED_CyberDelivery_AnimBP
target_skeleton=&/F_MED_CyberDelivery_Skeleton
graphs [
 AnimGraph {nodes [
   n1 = "Transform (Modify) Bone - Bone: base_belt_lower_sd_l"!(
    ComponentPose=Local To Component(LocalPose=Use cached pose 'BeltCache'())
    bAlphaBoolEnabled:bool=True
    Translation="0.316,-0.763,0.000"
    Rotation="0,0,0"
    Scale="1,1,1"
   ) -> (Pose) {
    alphaInputType=Bool
    rotationMode=BMM_Additive
    rotationSpace=BCS_BoneSpace
    scaleMode=BMM_Additive
    scaleSpace=BCS_BoneSpace
    translationMode=BMM_Additive
    translationSpace=BCS_BoneSpace
    alphaBoolBlend {blendInTime=0.2, blendOption=Linear, blendOutTime=0.2}
    alphaScaleBiasClamp {bInterpResult=true, interpSpeedIncreasing=5}
    boneToModify {boneName=base_belt_lower_sd_l}
    translation=<0.316 -0.763>
   }
   Output Pose(Result=Component To Local(ComponentPose=@n1)) {layerGroup=None}
   n2 = Control Rig(Source=Copy Pose From Mesh() {bUseAttachedParent=true}) {
    alphaBoolBlend {blendOption=Linear}
    controlRigAssetReference {blueprintRigClass=&/F_MED_CyberDelivery_CtrlRig.F_MED_CyberDelivery_CtrlRig_C}
   }
   Save cached pose 'Start Cache'(Pose=@n2) {cacheName=Start Cache}
   n3 = "Transform (Modify) Bone - Bone: base_belt_area_l"(
    ComponentPose=Local To Component(LocalPose=Use cached pose 'Start Cache'())
    Alpha=1
    Translation="-0.000,0.998,1.158"
    Rotation="0.000,-0.000,-4.450"
    Scale="1,1,1"
   ) {
    rotationMode=BMM_Additive
    rotationSpace=BCS_BoneSpace
    scaleMode=BMM_Additive
    scaleSpace=BCS_BoneSpace
    translationMode=BMM_Additive
    translationSpace=BCS_BoneSpace
    alphaBoolBlend {blendInTime=0.2, blendOption=Linear, blendOutTime=0.2}
    alphaScaleBiasClamp {bInterpResult=true}
    boneToModify {boneName=base_belt_area_l}
    rotation=<rot 0 0 -4.45>
    translation {y=0.998, z=1.1581}
   }
   n4 = "Transform (Modify) Bone - Bone: base_belt_lower_fr_l"(ComponentPose=@n3, Alpha=1, Translation="-0.755,0.937,-0.061", Rotation="-5.218,10.899,11.371", Scale="1,1,1") {
    rotationMode=BMM_Additive
    rotationSpace=BCS_BoneSpace
    scaleMode=BMM_Additive
    scaleSpace=BCS_BoneSpace
    translationMode=BMM_Additive
    translationSpace=BCS_BoneSpace
    alphaBoolBlend {blendInTime=0.2, blendOption=Linear, blendOutTime=0.2}
    alphaScaleBiasClamp {bInterpResult=true}
    boneToModify {boneName=base_belt_lower_fr_l}
    rotation=<rot -5.218 10.899 11.371>
    translation=<-0.755 0.937 -0.061>
   }
   n5 = "Transform (Modify) Bone - Bone: base_belt_lower_bk_l"(ComponentPose=@n4, Alpha=1, Translation="-1.972,-0.000,0.000", Rotation="-0.000,0.000,-18.649", Scale="1,1,1") {
    rotationMode=BMM_Additive
    rotationSpace=BCS_BoneSpace
    scaleMode=BMM_Additive
    scaleSpace=BCS_BoneSpace
    translationMode=BMM_Additive
    translationSpace=BCS_BoneSpace
    alphaBoolBlend {blendInTime=0.2, blendOption=Linear, blendOutTime=0.2}
    alphaScaleBiasClamp {bInterpResult=true}
    boneToModify {boneName=base_belt_lower_bk_l}
    rotation=<rot 0 0 -18.649>
    translation {x=-1.972}
   }
   n6 = "Transform (Modify) Bone - Bone: base_belt_c"(ComponentPose=@n5, Alpha=1, Translation="0.547,0.051,-0.050", Rotation="0.036,5.350,-0.470", Scale="1,1,1") {
    rotationMode=BMM_Additive
    rotationSpace=BCS_BoneSpace
    scaleMode=BMM_Additive
    scaleSpace=BCS_BoneSpace
    translationMode=BMM_Additive
    translationSpace=BCS_BoneSpace
    alphaBoolBlend {blendInTime=0.2, blendOption=Linear, blendOutTime=0.2}
    alphaScaleBiasClamp {bInterpResult=true}
    boneToModify {boneName=base_belt_c}
    rotation=<rot 0.0361 5.35 -0.47>
    translation=<0.547 0.051 -0.05>
   }
   n7 = "Transform (Modify) Bone - Bone: base_pistol_r"!(ComponentPose=@n6, bAlphaBoolEnabled:bool=True, Translation="0.331,0.546,0.051", Rotation="8.700,0.000,0.000", Scale="1,1,1") -> (Pose) {
    alphaInputType=Bool
    rotationMode=BMM_Additive
    rotationSpace=BCS_BoneSpace
    scaleMode=BMM_Additive
    scaleSpace=BCS_BoneSpace
    translationMode=BMM_Additive
    translationSpace=BCS_BoneSpace
    alphaBoolBlend {blendInTime=0.2, blendOption=Linear, blendOutTime=0.2}
    alphaScaleBiasClamp {bInterpResult=true, interpSpeedIncreasing=5}
    boneToModify {boneName=base_pistol_r}
    rotation=<rot 8.7 0 0>
    translation=<0.331 0.546 0.051>
   }
   n8 = "Transform (Modify) Bone - Bone: base_pistol_r"!(ComponentPose=@n7, Alpha:float=1, Translation="0,0,0", Rotation="-58.397,0.000,-0.000", Scale="1,1,1") -> (Pose) {
    rotationMode=BMM_Additive
    rotationSpace=BCS_BoneSpace
    scaleMode=BMM_Additive
    scaleSpace=BCS_BoneSpace
    translationMode=BMM_Additive
    translationSpace=BCS_BoneSpace
    alphaBoolBlend {blendInTime=0.2, blendOption=Linear, blendOutTime=0.2}
    alphaScaleBiasClamp {bInterpResult=true, interpSpeedIncreasing=5}
    boneToModify {boneName=base_pistol_r}
    rotation {pitch=-58.397, yaw=0}
   }
   n9 = "Transform (Modify) Bone - Bone: base_pistol_r"(ComponentPose=@n8, Alpha=1, Translation="0,0,0", Rotation="-10,-11.65,29.147", Scale="1,1,1") {
    rotationMode=BMM_Additive
    rotationSpace=BCS_BoneSpace
    scaleMode=BMM_Additive
    scaleSpace=BCS_BoneSpace
    translationMode=BMM_Additive
    translationSpace=BCS_BoneSpace
    alphaBoolBlend {blendInTime=0.2, blendOption=Linear, blendOutTime=0.2}
    alphaScaleBiasClamp {bInterpResult=true}
    boneToModify {boneName=base_pistol_r}
    rotation=<rot -10 -11.65 29.147>
   }
   Save cached pose 'BeltCache'(Pose=Component To Local(ComponentPose=@n9)) {cacheName=BeltCache}
  ]}
 EventGraph {nodes [
   n1 = Event BlueprintUpdateAnimation()
   > Set DeltaTimeX(DeltaTimeX:double=@n1.DeltaTimeX) -> (Output_Get:double)
   > n2 = Sequence!(execute) -> (then_0, then_1, then_2:Exec, then_3:Exec, then_4:Exec, then_5:Exec, then_6:Exec)
   n3 = PlayerMovement(execute=@n2.then_2, AnimInstance=Self-Reference(), XDotProductCoefficient=100, YDotProductCoefficient=100, ZDotProductCoefficient=100)
   n4 = BreakVector(InVec=MakeVector(X=@n3, Y=@n3.YMovement))
   > Set Right(Right:bool="float > float"(A=@n4, B=50)) -> (Output_Get:bool)
   > Set Forward(Forward:bool="float > float"(A=@n4.Y, B=40)) -> (Output_Get:bool)
   n5 = Get bIsMoving2D()
   n6 = Branch(execute=@n2.then_1, Condition=AND Boolean(A=Get bIsCrouching(), B=NOT Boolean(A=@n5)))
   > Set Alpha_CrouchIdle(Alpha_CrouchIdle:double=1) -> (Output_Get:double)
   Set Alpha_CrouchIdle(execute=@n6.else, Alpha_CrouchIdle=0)
   n7 = GetOwnerSkeletalMeshComponent()
   n8 = BreakRotator(InRot=BreakTransform(InTransform=GetDeltaTransformFromRefPose(self=@n7, BoneName=pelvis, BaseName=thigh_r)).Rotation)
   n9 = Branch(
    execute=@n2.then_3
    Condition=AND Boolean!(A="float > float"(A=@n8, B=22), B=Get bIsCrouching(), C:bool=Get Right() -> (Right:bool)) -> (ReturnValue)
   )
   > Set Alpha_LegRotateExtremSideInCrouch(Alpha_LegRotateExtremSideInCrouch:double=1) -> (Output_Get:double)
   Set Alpha_LegRotateExtremSideInCrouch(execute=@n9.else, Alpha_LegRotateExtremSideInCrouch=0)
   n10 = Branch(
    execute=@n2.then_4
    Condition=AND Boolean!(A="float > float"(A=@n8, B=17), B=Get Forward() -> (Forward:bool), C=Get bIsCrouching(), D:bool=Get Right()) -> (ReturnValue)
   )
   > Set Alpha_LegRotateExtremSideForwardInCrouch(Alpha_LegRotateExtremSideForwardInCrouch:double=1) -> (Output_Get:double)
   Set Alpha_LegRotateExtremSideForwardInCrouch(execute=@n10.else, Alpha_LegRotateExtremSideForwardInCrouch=0)
  ]}
]
variables [
 Alpha_CrouchIdle {type=number}
 Alpha_LegRotateExtremSideForwardInCrouch {type=number}
 Alpha_LegRotateExtremSideInCrouch {type=number}
 DeltaTimeX {type=number}
 Forward {type=boolean}
 Right {type=boolean}
 Skel Mesh {type=SkeletalMeshComponent}
]
