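# Control Rig for the F_MED_CyberDelivery face accessory mesh that corrects hair bone placement as the female character's pose and movement state change.
# Spherical pose readers on head and upper arms, plus a neck-to-head rotation delta, write weight variables.
# These weights, multiplied by state inputs like crouch, floating and weapon selection, blend additive bone offsets, with branches for skydiving and crouch.
# Parent constraints partly pull back hair toward neck and spine.
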ControlRigBlueprint /BRCosmetics/Characters/Player/Female/Medium/Bodies/F_MED_CyberDelivery/Meshes/Parts/F_MED_CyberDelivery_FaceAcc_CtrlRig
graphs [
 Multiply_10_ContainedGraph {nodes [
   n1 = RigVMFunctionEntryNode()
   RigVMFunctionReturnNode(Result=&RigVMFunction_MathDoubleMul(B=@n1.C, A=&RigVMFunction_MathDoubleMul(A=@n1.A, B=@n1.B).Result).Result)
  ]}
 Multiply_10_ContainedGraph {nodes [
   n1 = RigVMFunctionEntryNode()
   RigVMFunctionReturnNode(Result=&RigVMFunction_MathDoubleMul(B=@n1.C, A=&RigVMFunction_MathDoubleMul(A=@n1.A, B=@n1.B).Result).Result)
  ]}
 Multiply_7_ContainedGraph {nodes [
   n1 = RigVMFunctionEntryNode()
   RigVMFunctionReturnNode(Result=&RigVMFunction_MathDoubleMul(B=@n1.C, A=&RigVMFunction_MathDoubleMul(A=@n1.A, B=@n1.B).Result).Result)
  ]}
 Multiply_9_ContainedGraph {nodes [
   n1 = RigVMFunctionEntryNode()
   RigVMFunctionReturnNode(Result=&RigVMFunction_MathDoubleMul(B=@n1.C, A=&RigVMFunction_MathDoubleMul(A=@n1.A, B=@n1.B).Result).Result)
  ]}
 New Function_ContainedGraph {nodes [
   n1 = RigVMFunctionEntryNode()
   Clamp = &RigVMFunction_MathIntClamp(Value=@n1.primAxis)
   RigVMFunction_MathBoolToInteger = &RigVMFunction_MathBoolToInteger(Value=@n1.onlyTwist)
   RigVMFunction_MathQuaternionSwingTwist = &RigVMFunction_MathQuaternionSwingTwist(
    Input=&RigUnit_GetRelativeTransformForItem(
     Child=(Type=Bone,Name="None")
     Parent=(Type=Bone,Name="None")
     bChildInitial=False
     bParentInitial=False
     Parent.Type=@n1.ParentType
     Child.Name=@n1.child
     Parent.Name=@n1.parent
    ).RelativeTransform.Rotation
    TwistAxis=@n1.twistAxis
   )
   RigVMFunction_MathQuaternionToEuler = &RigVMFunction_MathQuaternionToEuler(
    Value=RigVMDispatchNode(
     Values=((),())
     Index=@RigVMFunction_MathBoolToInteger.Result
     Values.0=@RigVMFunction_MathQuaternionSwingTwist.Swing
     Values.1=@RigVMFunction_MathQuaternionSwingTwist.Twist
    ) {template="DISPATCH_RigVMDispatch_SelectInt32(in Index,in Values,out Result)"}.Result
    RotationOrder=@n1.rotationOrder
   )
   DISPATCH_RigVMDispatch_SelectInt32 = RigVMDispatchNode(
    Values=(0,0,0)
    Index=@Clamp.Result
    Values.0=@RigVMFunction_MathQuaternionToEuler.Result.X
    Values.1=@RigVMFunction_MathQuaternionToEuler.Result.Y
    Values.2=@RigVMFunction_MathQuaternionToEuler.Result.Z
   ) {template="DISPATCH_RigVMDispatch_SelectInt32(in Index,in Values,out Result)"}
   Remap = &RigVMFunction_MathDoubleRemap(
    TargetMaximum=1
    TargetMinimum=0
    bClamp=False
    Value=@DISPATCH_RigVMDispatch_SelectInt32.Result
    SourceMaximum=@n1.max
    SourceMinimum=@n1.min
   )
   > RigVMFunctionReturnNode(ouputWeight=&RigVMFunction_MathDoubleClamp(Value=@Remap.Result).Result)
  ]}
 New Function_ContainedGraph {nodes [
   n1 = RigVMFunctionEntryNode()
   RigVMFunction_ControlFlowBranch = &RigVMFunction_ControlFlowBranch(Condition=@n1.Controls Debug, ExecuteContext=@n1.ExecuteContext)
   RigVMFunctionReturnNode(ExecuteContext=@RigVMFunction_ControlFlowBranch.Completed)
   RigUnit_GetTransform = &RigUnit_GetTransform(Item=(Type=Control,Name="head_ctrl"), bInitial=False)
   RigUnit_SetTransform = &RigUnit_SetTransform(
    Item=(Type=Bone,Name="head")
    Weight=1
    bInitial=False
    Value=@RigUnit_GetTransform.Transform
    ExecutePin=@RigVMFunction_ControlFlowBranch.True
   )
   > &RigUnit_SetTransform(
    Item=(Type=Bone,Name="clavicle_l")
    Weight=1
    bInitial=False
    Value=&RigUnit_GetTransform(Item=(Type=Control,Name="clavicle_l_ctrl"), bInitial=False).Transform
   )
   > &RigUnit_SetTransform(
    Item=(Type=Bone,Name="clavicle_r")
    Weight=1
    bInitial=False
    Value=&RigUnit_GetTransform(Item=(Type=Control,Name="clavicle_r_ctrl"), bInitial=False).Transform
   )
   > &RigUnit_SetTransform(
    Item=(Type=Bone,Name="upperarm_l")
    Weight=1
    bInitial=False
    Value=&RigUnit_GetTransform(Item=(Type=Control,Name="upperarm_l_ctrl"), bInitial=False).Transform
   )
   > &RigUnit_SetTransform(
    Item=(Type=Bone,Name="upperarm_r")
    Weight=1
    bInitial=False
    Value=&RigUnit_GetTransform(Item=(Type=Control,Name="upperarm_r_ctrl"), bInitial=False).Transform
   )
  ]}
 New Function_ContainedGraph {nodes [
   n1 = RigVMFunctionEntryNode()
   RigVMFunction_MathQuaternionToEuler = &RigVMFunction_MathQuaternionToEuler(
    RotationOrder=ZYX
    Value=&RigUnit_GetRelativeTransformForItem(
     Child=(Type=Bone,Name="hand_l")
     Parent=(Type=Bone,Name="lowerarm_l")
     bChildInitial=False
     bParentInitial=False
     Child.Name=@n1.Child
     Parent.Name=@n1.Parent
    ).RelativeTransform.Rotation
   )
   Remap = &RigVMFunction_MathDoubleRemap(
    TargetMaximum=1
    TargetMinimum=0
    Value=@RigVMFunction_MathQuaternionToEuler.Result.Z
    SourceMaximum=@n1.Source Maximum
    SourceMinimum=@n1.Source Minimum
   )
   > RigVMFunctionReturnNode(Delta Z=@Remap.Result)
  ]}
 RigVMFunctionLibrary {nodes [
   Controls = RigVMCollapseNode(Controls Debug=False, ExecuteContext=())
   Delta Rotation Z = RigVMCollapseNode(ExecuteContext=())
   DeltaRotation = RigVMCollapseNode(ExecuteContext=(), debuigRotation=False, onlyTwist=False, twistAxis=(X=0,Y=0,Z=0))
  ]}
 RigVMModel {nodes [
   &RigUnit_ModifyTransforms(
    ExecutePin=()
    ItemToModify=((Item=(Type=Bone,Name="base_hair_area_bk_r"),Transform=(Rotation=(X=-0.1204,Y=0.067,Z=-0.0081,W=0.9904),Translation=(X=5.0973,Y=0.1945,Z=2.9153),Scale3D=(X=1,Y=1,Z=1))))
    Weight=1
    WeightMinimum=0
   )
   > &RigUnit_ModifyTransforms(
    ItemToModify="((Item=(Type=Bone,Name=\"dyn_hair_bk_a_l\"),Transform=(Rotation=(X=0,Y=0,Z=-0.2242,W=0.9745),Translation=(X=0,Y=-2.0381,Z=0),Scale3D=(X=1,Y=1,Z=1))),(Item=(Type=Bone,Name=\"base_hair_area_bk_l\"),Transform=(Rotation=(X=0,Y=0.4699,Z=0,W=0.8827),Translation=(X=4 ...(truncated)"
    Weight=1
    WeightMinimum=0
   )
   RigUnit_BeginExecution = &RigUnit_BeginExecution()
   Controls = RigVMFunctionReferenceNode(Controls Debug=false, ExecuteContext=@RigUnit_BeginExecution.ExecutePin)
   RigVMFunction_Sequence = &RigVMFunction_Sequence(ExecuteContext=@Controls.ExecuteContext)
   Delta Rotation Z = RigVMFunctionReferenceNode(Child=head, Parent=neck_01, Source Maximum=-55, Source Minimum=-25, ExecuteContext=@RigVMFunction_Sequence.A)
   n1 = RigVMVariableNode() {variable=Crouching}
   > RigVMVariableNode(Value=&RigVMFunction_MathDoubleMul(
     A=@"Delta Rotation Z".Delta Z
     B=&RigVMFunction_MathDoubleRemap(SourceMaximum=1, SourceMinimum=0, TargetMaximum=0, TargetMinimum=1, bClamp=False, Value=@n1.Value).Result
    ).Result) {variable=IsHeadUpDelta}
   > n2 = &RigUnit_SphericalPoseReader(
    ActiveRegionScaleFactors=(PositiveWidth=3,NegativeWidth=0.3,PositiveHeight=6,NegativeHeight=6)
    ActiveRegionSize=0.1
    Debug=(bDrawDebug=false,DebugScale=25,DebugSegments=20,DebugThickness=0.25)
    DriverAxis=(X=1,Y=0,Z=0)
    DriverItem=(Type=Bone,Name="head")
    FalloffRegionScaleFactors=(PositiveWidth=-4,NegativeWidth=1.2,PositiveHeight=-1.5,NegativeHeight=-1.5)
    FalloffSize=0.25
    OptionalParentItem=(Type=Bone,Name="spine_03")
    RotationOffset=(X=90,Y=0,Z=-140)
   )
   > RigVMVariableNode(Value=@n2.OutputParam) {variable=HeadRotationUp}
   > n3 = &RigUnit_SphericalPoseReader(
    ActiveRegionScaleFactors=(PositiveWidth=3,NegativeWidth=3,PositiveHeight=3,NegativeHeight=3)
    ActiveRegionSize=0.08
    Debug=(bDrawDebug=false,DebugScale=25,DebugSegments=20,DebugThickness=0.25)
    DriverAxis=(X=1,Y=0,Z=0)
    DriverItem=(Type=Bone,Name="upperarm_l")
    FalloffRegionScaleFactors=(PositiveWidth=2.1,NegativeWidth=2,PositiveHeight=2,NegativeHeight=2)
    FalloffSize=0.37
    OptionalParentItem=(Type=Bone,Name="spine_04")
    RotationOffset=(X=0,Y=0,Z=0)
   )
   > RigVMVariableNode(Value=@n3.OutputParam) {variable=UpperArmUpL}
   > n4 = &RigUnit_SphericalPoseReader(
    ActiveRegionScaleFactors=(PositiveWidth=3,NegativeWidth=3,PositiveHeight=3,NegativeHeight=3)
    ActiveRegionSize=0.08
    Debug=(bDrawDebug=false,DebugScale=25,DebugSegments=20,DebugThickness=0.25)
    DriverAxis=(X=-1,Y=0,Z=0)
    DriverItem=(Type=Bone,Name="upperarm_r")
    FalloffRegionScaleFactors=(PositiveWidth=2.1,NegativeWidth=2,PositiveHeight=2,NegativeHeight=2)
    FalloffSize=0.37
    OptionalParentItem=(Type=Bone,Name="spine_04")
    RotationOffset=(X=0,Y=180,Z=0)
   )
   > RigVMVariableNode(Value=@n4.OutputParam) {variable=UpperArmUpR}
   > n5 = &RigUnit_SphericalPoseReader(
    ActiveRegionScaleFactors=(PositiveWidth=3,NegativeWidth=0.6,PositiveHeight=6,NegativeHeight=6)
    ActiveRegionSize=0.1
    Debug=(bDrawDebug=false,DebugScale=25,DebugSegments=20,DebugThickness=0.25)
    DriverAxis=(X=1,Y=0,Z=0)
    DriverItem=(Type=Bone,Name="head")
    FalloffRegionScaleFactors=(PositiveWidth=-4,NegativeWidth=1.8,PositiveHeight=-1.5,NegativeHeight=-1.5)
    FalloffSize=0.25
    OptionalParentItem=(Type=Bone,Name="spine_03")
    RotationOffset=(X=90,Y=0,Z=-160)
   )
   > RigVMVariableNode(Value=@n5.OutputParam) {variable=HeadRotationUpCrouch}
   n6 = &RigVMFunction_MathDoubleRemap(SourceMaximum=1, SourceMinimum=0, TargetMaximum=0.4, TargetMinimum=0, Value=RigVMVariableNode() {variable=HeadRotationUpCrouch}.Value)
   > &RigUnit_ParentConstraint(
    AdvancedSettings=(RotationOrderForFilter=XZY)
    Child=(Type=Bone,Name="dyn_hair_area_bk_l")
    Parents=((Item=(Type=Bone,Name="neck_01"),Weight=0.1822),(Item=(Type=Bone,Name="spine_05"),Weight=0.268),(Item=(Type=Bone,Name="neck_02"),Weight=0.5497))
    Weight=@n6.Result
   )
   > n7 = &RigUnit_ParentConstraint(
    AdvancedSettings=(RotationOrderForFilter=XZY)
    Child=(Type=Bone,Name="dyn_hair_area_bk_r")
    Parents=((Item=(Type=Bone,Name="spine_03"),Weight=0.0007),(Item=(Type=Bone,Name="neck_01"),Weight=0.1873),(Item=(Type=Bone,Name="spine_05"),Weight=0.2875),(Item=(Type=Bone,Name="neck_02"),Weight=0.5245))
    Weight=@n6.Result
   )
   n8 = &RigVMFunction_ControlFlowBranch(ExecuteContext=@n7.ExecutePin, Condition=RigVMVariableNode() {variable=Skydiving}.Value)
   n9 = &RigUnit_SphericalPoseReader(
    ActiveRegionScaleFactors=(PositiveWidth=1,NegativeWidth=1.5,PositiveHeight=1,NegativeHeight=0.9)
    ActiveRegionSize=0.1
    Debug=(bDrawDebug=false,bDraw2D=False,bDrawLocalAxes=False,DebugScale=25,DebugSegments=20,DebugThickness=0.25)
    DriverAxis=(X=-1,Y=1,Z=0)
    DriverItem=(Type=Bone,Name="head")
    FalloffRegionScaleFactors=(PositiveWidth=1,NegativeWidth=4,PositiveHeight=1,NegativeHeight=0.6)
    FalloffSize=0.2
    FlipHeightScaling=False
    FlipWidthScaling=False
    OptionalParentItem=(Type=Bone,Name="spine_05")
    RotationOffset=(X=120,Y=0,Z=0)
    ExecutePin=@n8.True
   )
   > &RigUnit_ModifyTransforms(
    ItemToModify="((Item=(Type=Bone,Name=\"base_hair_area_bk_l\"),Transform=(Rotation=(X=0.1211,Y=-0.1969,Z=-0.0706,W=0.9703),Translation=(X=0,Y=0,Z=0),Scale3D=(X=1,Y=1,Z=1))),(Item=(Type=Bone,Name=\"base_hair_area_bk_r\"),Transform=(Rotation=(X=-0.0888,Y=-0.1565,Z=-0.1059,W=0. ...(truncated)"
    WeightMinimum=0
    Weight=@n9.OutputParam
   )
   > n10 = &RigUnit_SphericalPoseReader(
    ActiveRegionScaleFactors=(PositiveWidth=1,NegativeWidth=1.5,PositiveHeight=0.9,NegativeHeight=1)
    ActiveRegionSize=0.1
    Debug=(bDrawDebug=false,bDraw2D=False,bDrawLocalAxes=False,DebugScale=25,DebugSegments=20,DebugThickness=0.25)
    DriverAxis=(X=-1,Y=1,Z=0)
    DriverItem=(Type=Bone,Name="head")
    FalloffRegionScaleFactors=(PositiveWidth=1,NegativeWidth=4,PositiveHeight=0.6,NegativeHeight=1)
    FalloffSize=0.2
    FlipHeightScaling=False
    FlipWidthScaling=False
    OptionalParentItem=(Type=Bone,Name="spine_05")
    RotationOffset=(X=60,Y=0,Z=0)
   )
   > &RigUnit_ModifyTransforms(
    ItemToModify="((Item=(Type=Bone,Name=\"base_hair_area_bk_l\"),Transform=(Rotation=(X=0.0887,Y=-0.0117,Z=-0.001,W=0.996),Translation=(X=-2.5145,Y=0,Z=0),Scale3D=(X=1,Y=1,Z=1))),(Item=(Type=Bone,Name=\"base_hair_area_bk_r\"),Transform=(Rotation=(X=0,Y=-0.0683,Z=0,W=0.9977),Tr ...(truncated)"
    WeightMinimum=0
    Weight=@n10.OutputParam
   )
   n11 = &RigVMFunction_ControlFlowBranch(Condition=RigVMVariableNode() {variable=CrouchBool}.Value, ExecuteContext=@RigVMFunction_Sequence.B)
   Remap = &RigVMFunction_MathDoubleRemap(
    SourceMaximum=1
    SourceMinimum=0
    TargetMaximum=0.6
    TargetMinimum=0
    bClamp=False
    Value=RigVMVariableNode() {variable=HeadRotationUpCrouch}.Value
   )
   &RigUnit_ModifyTransforms(
    ItemToModify="((Item=(Type=Bone,Name=\"base_hair_area_bk_l\"),Transform=(Rotation=(X=-0.0294,Y=-0.2544,Z=0.1677,W=0.952),Translation=(X=0.6902,Y=-3.0828,Z=3.3319),Scale3D=(X=1,Y=1,Z=1))),(Item=(Type=Bone,Name=\"base_hair_area_bk_r\"),Transform=(Rotation=(X=0,Y=0.4148,Z=0,W= ...(truncated)"
    WeightMinimum=0
    Weight=&RigVMFunction_MathDoubleMul(A=@Remap.Result, B=RigVMVariableNode() {variable=Crouching}.Value).Result
    ExecutePin=@n11.Completed
   )
   Multiply_9 = RigVMAggregateNode(
    A=&RigVMFunction_MathDoubleRemap(SourceMaximum=0.77, SourceMinimum=0, TargetMaximum=1, TargetMinimum=0, Value=RigVMVariableNode() {variable=UpperArmUpL}.Value).Result
    B=RigVMVariableNode() {variable=SelectedPistol}.Value
    C=RigVMVariableNode() {variable=Standing}.Value
   )
   > &RigUnit_ModifyTransforms(
    ItemToModify="((Item=(Type=Bone,Name=\"base_hair_area_bk_l\"),Transform=(Rotation=(X=0,Y=0,Z=0.1357,W=0.9908),Translation=(X=-0.4451,Y=-0.3805,Z=1.0203),Scale3D=(X=1,Y=1,Z=1))),(Item=(Type=Bone,Name=\"base_hair_area_sd_l\"),Transform=(Rotation=(X=0,Y=-0.1643,Z=0,W=0.9864),T ...(truncated)"
    WeightMinimum=0
    Weight=@Multiply_9.Result
   )
   Multiply_8 = RigVMAggregateNode(
    B=RigVMVariableNode() {variable=Crouching}.Value
    C=RigVMVariableNode() {variable=SelectedPickaxe}.Value
    A=RigVMVariableNode() {variable=HeadRotationUpCrouch}.Value
   )
   > &RigUnit_ModifyTransforms(
    ItemToModify="((Item=(Type=Bone,Name=\"base_hair_area_bk_l\"),Transform=(Rotation=(X=0.0386,Y=0.1515,Z=0.0294,W=0.9873),Translation=(X=0.658,Y=-0.2264,Z=2.6501),Scale3D=(X=1,Y=1,Z=1))),(Item=(Type=Bone,Name=\"base_hair_area_bk_r\"),Transform=(Rotation=(X=0.0396,Y=-0.3326,Z= ...(truncated)"
    WeightMinimum=0
    Weight=@Multiply_8.Result
   )
   > &RigUnit_ModifyTransforms(
    ItemToModify="((Item=(Type=Bone,Name=\"base_hair_area_bk_l\"),Transform=(Rotation=(X=0,Y=0,Z=0,W=1),Translation=(X=0,Y=0,Z=2),Scale3D=(X=1,Y=1,Z=1))),(Item=(Type=Bone,Name=\"base_hair_area_bk_r\"),Transform=(Rotation=(X=0,Y=0,Z=0,W=1),Translation=(X=0,Y=0,Z=2),Scale3D=(X=1, ...(truncated)"
    WeightMinimum=0
    Weight=RigVMVariableNode() {variable=TacticalSprinting}.Value
   )
   > n12 = &RigUnit_ModifyTransforms(
    ItemToModify="((Item=(Type=Bone,Name=\"base_hair_bk_c_l\"),Transform=(Rotation=(X=0,Y=0,Z=0,W=1),Translation=(X=-2,Y=0,Z=0),Scale3D=(X=1,Y=1,Z=1))),(Item=(Type=Bone,Name=\"base_hair_bk_d_l\"),Transform=(Rotation=(X=0,Y=0,Z=0,W=1),Translation=(X=0,Y=0,Z=0),Scale3D=(X=1,Y=1,Z ...(truncated)"
    WeightMinimum=0
    Weight=RigVMVariableNode() {variable=Crouching}.Value
   )
   RigVMFunction_MathBoolNot = &RigVMFunction_MathBoolNot(Value=RigVMVariableNode() {variable=Skydiving}.Value)
   RigVMFunction_ControlFlowBranch = &RigVMFunction_ControlFlowBranch(Condition=@RigVMFunction_MathBoolNot.Result, ExecuteContext=@n12.ExecutePin)
   &RigUnit_ModifyTransforms(
    ItemToModify="((Item=(Type=Bone,Name=\"base_hair_area_sd_r\"),Transform=(Rotation=(X=0.2647,Y=-0.1257,Z=-0.0348,W=0.9555),Translation=(X=-0.4177,Y=-1.2294,Z=-0.6162),Scale3D=(X=1,Y=1,Z=1))),(Item=(Type=Bone,Name=\"base_hair_area_bk_r\"),Transform=(Rotation=(X=0,Y=-0.0799,Z= ...(truncated)"
    WeightMinimum=0
    Weight=&RigVMFunction_MathDoubleMul(A=RigVMVariableNode() {variable=UpperArmUpR}.Value, B=RigVMVariableNode() {variable=Floating}.Value).Result
    ExecutePin=@RigVMFunction_ControlFlowBranch.Completed
   )
   > &RigUnit_ModifyTransforms(
    ItemToModify="((Item=(Type=Bone,Name=\"base_hair_area_sd_l\"),Transform=(Rotation=(X=-0.0209,Y=0,Z=0.0018,W=0.9998),Translation=(X=0,Y=0,Z=0),Scale3D=(X=1,Y=1,Z=1))),(Item=(Type=Bone,Name=\"base_hair_area_bk_l\"),Transform=(Rotation=(X=0.1119,Y=0.0143,Z=0.0232,W=0.9933),Tra ...(truncated)"
    WeightMinimum=0
    Weight=&RigVMFunction_MathDoubleMul(B=RigVMVariableNode() {variable=Floating}.Value, A=RigVMVariableNode() {variable=UpperArmUpL}.Value).Result
   )
   n13 = RigVMVariableNode() {variable=Floating}
   > &RigUnit_ModifyTransforms(
    ItemToModify="((Item=(Type=Bone,Name=\"base_hair_area_bk_r\"),Transform=(Rotation=(X=-0.1204,Y=0.067,Z=-0.0081,W=0.9904),Translation=(X=5.0973,Y=0.1945,Z=2.9153),Scale3D=(X=1,Y=1,Z=1))),(Item=(Type=Bone,Name=\"base_hair_bk_mid\"),Transform=(Rotation=(X=0,Y=0,Z=0,W=1),Transl ...(truncated)"
    WeightMinimum=0
    Weight=@n13.Value
   )
   > &RigUnit_ModifyTransforms(
    ItemToModify="((Item=(Type=Bone,Name=\"dyn_hair_bk_a_l\"),Transform=(Rotation=(X=0,Y=0,Z=-0.2242,W=0.9745),Translation=(X=0,Y=-2.0381,Z=0),Scale3D=(X=1,Y=1,Z=1))),(Item=(Type=Bone,Name=\"base_hair_area_bk_l\"),Transform=(Rotation=(X=0,Y=0.4699,Z=0,W=0.8827),Translation=(X=4 ...(truncated)"
    WeightMinimum=0
    Weight=@n13.Value
   )
   &RigUnit_ModifyTransforms(
    ItemToModify="((Item=(Type=Bone,Name=\"base_hair_area_sd_r\"),Transform=(Rotation=(X=0.0015,Y=-0.3865,Z=0.0576,W=0.9205),Translation=(X=-0.9564,Y=0.1994,Z=-0.9523),Scale3D=(X=1,Y=1,Z=1))),(Item=(Type=Bone,Name=\"base_hair_area_bk_r\"),Transform=(Rotation=(X=0,Y=0,Z=0,W=1),T ...(truncated)"
    WeightMinimum=0
    ExecutePin=@RigVMFunction_ControlFlowBranch.True
    Weight=RigVMVariableNode() {variable=UpperArmUpR}.Value
   )
   Multiply_5 = RigVMAggregateNode(
    A=RigVMVariableNode() {variable=UpperArmUpL}.Value
    B=RigVMVariableNode() {variable=NotTargeting}.Value
    C=RigVMVariableNode() {variable=NotSelectedPistol}.Value
   )
   > &RigUnit_ModifyTransforms(
    ItemToModify="((Item=(Type=Bone,Name=\"base_hair_area_bk_l\"),Transform=(Rotation=(X=0.1588,Y=-0.1639,Z=0.0267,W=0.9732),Translation=(X=-0.8075,Y=-0.0801,Z=0.3048),Scale3D=(X=1,Y=1,Z=1))),(Item=(Type=Bone,Name=\"base_hair_area_sd_l\"),Transform=(Rotation=(X=0,Y=-0.2672,Z=0, ...(truncated)"
    WeightMinimum=0
    Weight=@Multiply_5.Result
   )
   &RigUnit_ModifyTransforms(
    ItemToModify="((Item=(Type=Bone,Name=\"base_hair_area_bk_l\"),Transform=(Rotation=(X=0.0205,Y=-0.448,Z=-0.1314,W=0.8841),Translation=(X=-3.2568,Y=-0.0665,Z=1.2485),Scale3D=(X=1,Y=1,Z=1))),(Item=(Type=Bone,Name=\"base_hair_area_bk_r\"),Transform=(Rotation=(X=0,Y=-0.4155,Z=0, ...(truncated)"
    WeightMinimum=0
    ExecutePin=@n11.False
    Weight=RigVMVariableNode() {variable=IsHeadUpDelta}.Value
   )
   n14 = RigVMVariableNode() {variable=IsHeadUpDelta}
   > &RigUnit_ModifyTransforms(
    ItemToModify="((Item=(Type=Bone,Name=\"base_hair_area_bk_l\"),Transform=(Rotation=(X=0.0205,Y=-0.448,Z=-0.1314,W=0.8841),Translation=(X=-3.2568,Y=-0.0665,Z=1.2485),Scale3D=(X=1,Y=1,Z=1))),(Item=(Type=Bone,Name=\"base_hair_area_bk_r\"),Transform=(Rotation=(X=0,Y=-0.4155,Z=0, ...(truncated)"
    WeightMinimum=0
    Weight=&RigVMFunction_MathDoubleMul(
     A=&RigVMFunction_MathDoubleRemap(SourceMaximum=1, SourceMinimum=0, TargetMaximum=0, TargetMinimum=1, bClamp=False, Value=@n14.Value).Result
     B=RigVMVariableNode() {variable=Floating}.Value
    ).Result
   )
   &RigUnit_ModifyTransforms(
    ItemToModify="((Item=(Type=Bone,Name=\"base_hair_area_bk_l\"),Transform=(Rotation=(X=0.0205,Y=-0.448,Z=-0.1314,W=0.8841),Translation=(X=-1.5826,Y=-0.0739,Z=0.7047),Scale3D=(X=1,Y=1,Z=1))),(Item=(Type=Bone,Name=\"base_hair_area_bk_r\"),Transform=(Rotation=(X=0,Y=-0.4155,Z=0, ...(truncated)"
    WeightMinimum=0
    ExecutePin=@n11.True
    Weight=RigVMVariableNode() {variable=HeadRotationUp}.Value
   )
   &RigUnit_ModifyTransforms(
    ExecutePin=()
    ItemToModify="((Item=(Type=Bone,Name=\"dyn_hair_area_sd_l\"),Transform=(Rotation=(X=0,Y=0,Z=0.1489,W=0.9889),Translation=(X=-2.3618,Y=-1.3225,Z=-2.5243),Scale3D=(X=1,Y=1,Z=1))),(Item=(Type=Bone,Name=\"base_hair_area_bk_r\"),Transform=(Rotation=(X=-0.3063,Y=-0.584,Z=-0.1999, ...(truncated)"
    Weight=0
    WeightMinimum=0
   )
   Multiply_6 = RigVMAggregateNode(
    A=&RigVMFunction_MathDoubleRemap(
     SourceMaximum=0.25
     SourceMinimum=0
     TargetMaximum=0.7
     TargetMinimum=0
     bClamp=False
     Value=RigVMVariableNode() {variable=UpperArmUpR}.Value
    ).Result
    C=RigVMVariableNode() {variable=Crouching}.Value
    B=RigVMVariableNode() {variable=SelectedPistol}.Value
   )
  ]}
]
hierarchy [
 root {loc=<0 0 0>, rot=<0 0 0 1>}
 pelvis {parent=root, loc=<0 -0.5301 98.0571>, rot=<0.0127 0.707 -0.0127 -0.707>}
 spine_01 {parent=pelvis, loc=<6.1208 -0.9328 0>, rot=<0 0 0.034 -0.9994>}
 spine_02 {parent=spine_01, loc=<6.5997 -0.2095 0>, rot=<0 0 0.0238 -0.9997>}
 spine_03 {parent=spine_02, loc=<6.4914 -0.1237 0>, rot=<0 0 0.0149 -0.9999>}
 spine_04 {parent=spine_03, loc=<6.9598 0.0794 0>, rot=<0 0 -0.0132 -0.9999>}
 spine_05 {parent=spine_04, loc=<6.9987 0 0>, rot=<0 0 -0.0015 -1>}
 neck_01 {parent=spine_05, loc=<11.3138 0.1529 0>, rot=<0 0 -0.2003 -0.9797>}
 neck_02 {parent=neck_01, loc=<4.5627 0 0>, rot=<0 0 0.0504 -0.9987>}
 head {parent=neck_02, loc=<5.3914 0 0>, rot=<0 0 0.115 -0.9934>}
 base_hair_area_sd_r {parent=head, loc=<-0.4001 -4.6009 7.7253>, rot=<-0.0029 -0.0071 1 -0.0036>}
 dyn_hair_area_sd_r {parent=base_hair_area_sd_r, loc=<0 0 0>, rot=<0 0 0 1>}
 dyn_hair_sd_a_r {parent=dyn_hair_area_sd_r, loc=<2.0419 -1.1873 0.7414>, rot=<-0.1617 0.2026 0.0891 -0.9617>}
 dyn_hair_sd_b_r {parent=dyn_hair_area_sd_r, loc=<5.0479 -1.0165 -0.0273>, rot=<-0.1005 0.3728 0.0028 -0.9224>}
 dyn_hair_sd_d_r {parent=dyn_hair_area_sd_r, loc=<-1.1018 2.2728 0.7036>, rot=<-0.1268 0.3689 0.0402 -0.9199>}
 dyn_hair_sd_e_r {parent=dyn_hair_area_sd_r, loc=<2.4517 -2.8059 -0.6521>, rot=<-0.1877 0.1789 0.2184 -0.9408>}
 dyn_hair_sd_f_r {parent=dyn_hair_area_sd_r, loc=<3.1473 -2.532 -1.7364>, rot=<-0.2304 -0.0286 0.1727 -0.9572>}
 dyn_hair_sd_g_r {parent=dyn_hair_area_sd_r, loc=<5.358 0.8037 -1.5633>, rot=<-0.1402 0.1393 0.1133 -0.9737>}
 dyn_hair_sd_h_r {parent=dyn_hair_area_sd_r, loc=<3.735 2.8617 -1.1669>, rot=<-0.0987 0.2555 -0.0162 -0.9616>}
 base_hair_sd_c_r {parent=dyn_hair_area_sd_r, loc=<1.298 2.4168 0.185>, rot=<-0.0781 0.4817 -0.0244 -0.8725>}
 dyn_hair_sd_c_r {parent=base_hair_sd_c_r, loc=<0 0 0>, rot=<0 0 0 1>}
 base_hair_area_bk_l {parent=head, loc=<-1.273 -9.3656 -3.9031>, rot=<0.0558 -0.7663 0.6384 -0.0465>}
 dyn_hair_area_bk_l {parent=base_hair_area_bk_l, loc=<0 0 0>, rot=<0 0 0 1>}
 dyn_hair_bk_b_l {parent=dyn_hair_area_bk_l, loc=<2.5895 1.3014 0.5754>, rot=<0.2762 0.2026 -0.2601 -0.9028>}
 dyn_hair_bk_e_l {parent=dyn_hair_area_bk_l, loc=<2.8349 1.3463 -1.7282>, rot=<0.2717 0.2181 -0.2648 -0.8992>}
 dyn_hair_bk_f_l {parent=dyn_hair_area_bk_l, loc=<0.325 -0.3837 2.5145>, rot=<0.249 0.1386 -0.2501 -0.9253>}
 base_hair_bk_a_l {parent=dyn_hair_area_bk_l, loc=<2.5055 5.6601 -1.3721>, rot=<0.4568 0.1985 -0.3194 -0.8061>}
 dyn_hair_bk_a_l {parent=base_hair_bk_a_l, loc=<0 0 0>, rot=<0 0 0 1>}
 base_hair_bk_c_l {parent=dyn_hair_area_bk_l, loc=<4.2914 -3.0978 1.2756>, rot=<0.0466 0.3314 -0.2113 -0.9184>}
 dyn_hair_bk_c_l {parent=base_hair_bk_c_l, loc=<0 0 0>, rot=<0 0 0 1>}
 base_hair_bk_d_l {parent=dyn_hair_area_bk_l, loc=<4.0415 -0.8292 1.1491>, rot=<0.1181 0.308 -0.0905 -0.9397>}
 dyn_hair_bk_d_l {parent=base_hair_bk_d_l, loc=<0 0 0>, rot=<0 0 0 1>}
 base_hair_area_sd_l {parent=head, loc=<-1.8951 -5.4275 -6.6204>, rot=<-0.005 -0.9994 -0.0305 -0.0184>}
 dyn_hair_area_sd_l {parent=base_hair_area_sd_l, loc=<0 0 0>, rot=<0 0 0 1>}
 dyn_hair_sd_a_l {parent=dyn_hair_area_sd_l, loc=<3.0969 5.1925 0.6979>, rot=<0.3837 -0.0207 -0.4666 -0.7966>}
 dyn_hair_sd_b_l {parent=dyn_hair_area_sd_l, loc=<4.2964 4.995 1.5023>, rot=<0.168 0.1265 -0.3229 -0.9228>}
 dyn_hair_sd_d_l {parent=dyn_hair_area_sd_l, loc=<0.8117 3.074 3.3094>, rot=<0.253 0.4217 -0.1082 -0.864>}
 dyn_hair_sd_f_l {parent=dyn_hair_area_sd_l, loc=<-2.6968 -2.0871 1.1576>, rot=<0.1725 0.4148 -0.0447 -0.8923>}
 dyn_hair_sd_g_l {parent=dyn_hair_area_sd_l, loc=<2.6858 -0.5683 -1.812>, rot=<0.2154 0.3389 -0.1079 -0.9095>}
 dyn_hair_sd_h_l {parent=dyn_hair_area_sd_l, loc=<-0.4909 3.5703 2.2342>, rot=<0.2949 0.3936 -0.196 -0.8484>}
 base_hair_sd_c_l {parent=dyn_hair_area_sd_l, loc=<3.6234 1.7427 1.6413>, rot=<0.2954 0.2343 -0.1633 -0.9117>}
 dyn_hair_sd_c_l {parent=base_hair_sd_c_l, loc=<0 0 0>, rot=<0 0 0 1>}
 base_hair_sd_e_l {parent=dyn_hair_area_sd_l, loc=<0.887 0.5261 1.9129>, rot=<0.2655 0.2855 -0.0638 -0.9187>}
 dyn_hair_sd_e_l {parent=base_hair_sd_e_l, loc=<0 0 0>, rot=<0 0 0 1>}
 base_hair_bk_area_c {parent=head, loc=<4.3521 -11.5823 0.0596>, rot=<0.0916 -0.7055 0.6961 -0.097>}
 dyn_hair_bk_a_c {parent=base_hair_bk_area_c, loc=<0.4027 -0.7261 0>, rot=<0.0116 0.1208 0.0951 -0.988>}
 base_hair_area_bk_r {parent=head, loc=<-1.2677 -9.3629 5.4645>, rot=<0.0082 -0.7284 0.685 0.0137>}
 dyn_hair_area_bk_r {parent=base_hair_area_bk_r, loc=<0 0 0>, rot=<0 0 0 1>}
 dyn_hair_bk_a_r {parent=dyn_hair_area_bk_r, loc=<2.8846 -0.4758 0.9316>, rot=<-0.3456 0.3301 0.156 -0.8644>}
 dyn_hair_bk_b_r {parent=dyn_hair_area_bk_r, loc=<2.5812 3.1829 1.9807>, rot=<-0.0241 0.3883 -0.0595 -0.9193>}
 dyn_hair_bk_c_r {parent=dyn_hair_area_bk_r, loc=<2.8625 1.473 2.0683>, rot=<-0.211 0.3122 0.1869 -0.9072>}
 base_hair_bk_d_r {parent=dyn_hair_area_bk_r, loc=<-1.7061 -0.1569 -0.1868>, rot=<-0.4675 0.3306 0.2805 -0.7704>}
 dyn_hair_bk_d_r {parent=base_hair_bk_d_r, loc=<0 0 0>, rot=<0 0 0 1>}
 base_hair_bk_mid {parent=head, loc=<-1.0116 -11.2336 0>, rot=<0.0827 -0.7023 0.7023 -0.0827>}
 clavicle_l {parent=spine_05, loc=<5.4831 -0.1053 -3.2946>, rot=<0.0465 -0.7289 -0.009 -0.6829>}
 upperarm_l {parent=clavicle_l, loc=<9.8658 -1.2136 -0.4105>, rot=<0.0164 -0.3245 -0.0269 -0.9454>}
 lowerarm_l {parent=upperarm_l, loc=<26.465 -0.8345 0.259>, rot=<0.0036 0.0161 -0.2298 -0.9731>}
 hand_l {parent=lowerarm_l, loc=<22.7355 -0.3203 -0.6105>, rot=<-0.6784 -0.0223 0.0735 -0.7307>}
 thumb_01_l {parent=hand_l, loc=<1.3118 -0.3191 -2.459>, rot=<0.6546 -0.4042 0.0664 -0.6354>}
 thumb_02_l {parent=thumb_01_l, loc=<5.9336 -0.0448 0.0401>, rot=<-0.0037 0.0213 0.298 -0.9543>}
 thumb_03_l {parent=thumb_02_l, loc=<2.2223 0.3682 -0.0907>, rot=<0.0118 -0.013 0.0066 -0.9998>}
 index_metacarpal_l {parent=hand_l, loc=<1.6535 0.0183 -2.96>, rot=<0.0158 -0.0124 -0.0173 -0.9996>}
 index_01_l {parent=index_metacarpal_l, loc=<8.3008 0.0357 -0.0524>, rot=<-0.0044 0.0485 0.0924 -0.9945>}
 index_02_l {parent=index_01_l, loc=<3.3587 -0.3004 0.0452>, rot=<-0.002 0.0027 0.3225 -0.9466>}
 index_03_l {parent=index_02_l, loc=<2.5578 -0.123 -0.0252>, rot=<-0.002 0.0067 0.2558 -0.9667>}
 middle_metacarpal_l {parent=hand_l, loc=<2.1856 -0.1725 -0.5984>, rot=<0.006 0.0168 -0.0329 -0.9993>}
 middle_01_l {parent=middle_metacarpal_l, loc=<7.3275 -0.2447 0.0129>, rot=<-0.0208 0.0042 0.0858 -0.9961>}
 middle_02_l {parent=middle_01_l, loc=<4.0123 0.1401 0.1175>, rot=<-0.0218 0.0258 0.3512 -0.9357>}
 middle_03_l {parent=middle_02_l, loc=<2.6077 0.0534 0.1373>, rot=<0.0082 0.0188 0.1708 -0.9851>}
 ring_metacarpal_l {parent=hand_l, loc=<2.2074 -0.3625 1.3638>, rot=<-0.0284 0.0722 -0.0458 -0.9959>}
 ring_01_l {parent=ring_metacarpal_l, loc=<7.0295 -0.6197 -0.1446>, rot=<-0.0489 -0.0252 0.1892 -0.9804>}
 ring_02_l {parent=ring_01_l, loc=<3.6174 0.4647 -0.1027>, rot=<0 0 0.2806 -0.9598>}
 ring_03_l {parent=ring_02_l, loc=<2.4152 0.114 0.0199>, rot=<0 0 0.1189 -0.9929>}
 pinky_metacarpal_l {parent=hand_l, loc=<1.7616 -0.9753 3.3814>, rot=<-0.0397 0.0814 -0.069 -0.9935>}
 pinky_01_l {parent=pinky_metacarpal_l, loc=<7.1709 -0.4831 -0.1094>, rot=<-0.0654 -0.0035 0.2242 -0.9723>}
 pinky_02_l {parent=pinky_01_l, loc=<2.8274 0.0893 -0.1043>, rot=<0.0037 0.0011 0.2854 -0.9584>}
 pinky_03_l {parent=pinky_02_l, loc=<2.2046 0.0307 -0.0078>, rot=<-0.0002 0.0098 0.1445 -0.9895>}
 deform_wrist_dn_base_l {parent=hand_l, loc=<0 -1.5326 0>, rot=<0 0 0 -1>}
 deform_wrist_dn_l {parent=deform_wrist_dn_base_l, loc=<0 0 0>, rot=<0 0 0 1>}
 deform_wrist_up_base_l {parent=hand_l, loc=<0 2.3009 0>, rot=<0 0 0 -1>}
 deform_wrist_up_l {parent=deform_wrist_up_base_l, loc=<0 0 0>, rot=<0 0 0 1>}
 lowerarm_twist_02_l {parent=lowerarm_l, loc=<10.6707 0 0>, rot=<0 0 0 1>}
 lowerarm_twist_01_l {parent=lowerarm_l, loc=<18.316 0 0>, rot=<0 0 0 1>}
 deform_elbow_in_base_l {parent=lowerarm_l, loc=<0.1834 3.6062 -0.2572>, rot=<0 0 0 -1>}
 deform_elbow_in_l {parent=deform_elbow_in_base_l, loc=<0 0 0>, rot=<0 0 0 1>}
 deform_elbow_out_base_l {parent=lowerarm_l, loc=<0 -4 0>, rot=<0 0 0 1>}
 deform_elbow_out_l {parent=deform_elbow_out_base_l, loc=<0 0 0>, rot=<0 0 0 1>}
 upperarm_twist_01_l {parent=upperarm_l, loc=<7.2135 -0.1134 0>, rot=<0 0 0 1>}
 deform_upperarm_bk_base_l {parent=upperarm_twist_01_l, loc=<-6.4053 -4.3746 0.314>, rot=<0 0 0 -1>}
 deform_upperarm_bk_l {parent=deform_upperarm_bk_base_l, loc=<0 0 0>, rot=<0 0 0 1>}
 deform_upperarm_fr_base_l {parent=upperarm_twist_01_l, loc=<-5.6312 5.092 0.3354>, rot=<0 0 0 -1>}
 deform_upperarm_fr_l {parent=deform_upperarm_fr_base_l, loc=<0 0 0>, rot=<0 0 0 1>}
 deform_upperarm_up_base_l {parent=upperarm_twist_01_l, loc=<-8.891 0.4825 3.9517>, rot=<0 0 0 -1>}
 deform_upperarm_up_l {parent=deform_upperarm_up_base_l, loc=<0 0 0>, rot=<0 0 0 1>}
 upperarm_twist_02_l {parent=upperarm_l, loc=<19.7998 -0.1134 0>, rot=<0 0 0 1>}
 dyn_shoulder_pad_st_a_l {parent=clavicle_l, loc=<9.5142 -1.4496 3.6803>, rot=<0.0097 -0.3214 -0.0233 -0.9466>}
 dyn_shoulder_pad_st_b_l {parent=dyn_shoulder_pad_st_a_l, loc=<0 0 0>, rot=<0 0 0 1>}
 clavicle_r {parent=spine_05, loc=<5.4831 -0.1053 3.2946>, rot=<0.7289 0.0465 -0.6829 0.009>}
 upperarm_r {parent=clavicle_r, loc=<-9.8658 1.2136 0.4105>, rot=<0.0164 -0.3245 -0.0269 -0.9454>}
 lowerarm_r {parent=upperarm_r, loc=<-26.465 0.8345 -0.259>, rot=<0.0036 0.0161 -0.2298 -0.9731>}
 hand_r {parent=lowerarm_r, loc=<-22.7355 0.3203 0.6106>, rot=<-0.6784 -0.0223 0.0735 -0.7307>}
 thumb_01_r {parent=hand_r, loc=<-1.3118 0.3191 2.459>, rot=<0.6546 -0.4042 0.0664 -0.6354>}
 thumb_02_r {parent=thumb_01_r, loc=<-5.9336 0.0448 -0.0401>, rot=<-0.0037 0.0213 0.298 -0.9543>}
 thumb_03_r {parent=thumb_02_r, loc=<-2.2223 -0.3682 0.0907>, rot=<0.0118 -0.013 0.0066 -0.9998>}
 index_metacarpal_r {parent=hand_r, loc=<-1.6535 -0.0183 2.96>, rot=<0.0158 -0.0124 -0.0173 -0.9996>}
 index_01_r {parent=index_metacarpal_r, loc=<-8.3008 -0.0357 0.0524>, rot=<-0.0044 0.0485 0.0924 -0.9945>}
 index_02_r {parent=index_01_r, loc=<-3.3587 0.3004 -0.0452>, rot=<-0.002 0.0027 0.3225 -0.9466>}
 index_03_r {parent=index_02_r, loc=<-2.5578 0.123 0.0252>, rot=<-0.002 0.0067 0.2558 -0.9667>}
 middle_metacarpal_r {parent=hand_r, loc=<-2.1856 0.1725 0.5984>, rot=<0.006 0.0168 -0.0329 -0.9993>}
 middle_01_r {parent=middle_metacarpal_r, loc=<-7.3275 0.2447 -0.0129>, rot=<-0.0208 0.0042 0.0858 -0.9961>}
 middle_02_r {parent=middle_01_r, loc=<-4.0123 -0.1401 -0.1175>, rot=<-0.0218 0.0258 0.3512 -0.9357>}
 middle_03_r {parent=middle_02_r, loc=<-2.6077 -0.0534 -0.1373>, rot=<0.0082 0.0188 0.1708 -0.9851>}
 ring_metacarpal_r {parent=hand_r, loc=<-2.2074 0.3625 -1.3638>, rot=<-0.0284 0.0722 -0.0458 -0.9959>}
 ring_01_r {parent=ring_metacarpal_r, loc=<-7.0295 0.6197 0.1446>, rot=<-0.0489 -0.0252 0.1892 -0.9804>}
 ring_02_r {parent=ring_01_r, loc=<-3.6174 -0.4647 0.1027>, rot=<0 0 0.2806 -0.9598>}
 ring_03_r {parent=ring_02_r, loc=<-2.4152 -0.114 -0.0199>, rot=<0 0 0.1189 -0.9929>}
 pinky_metacarpal_r {parent=hand_r, loc=<-1.7616 0.9753 -3.3814>, rot=<-0.0397 0.0814 -0.069 -0.9935>}
 pinky_01_r {parent=pinky_metacarpal_r, loc=<-7.1709 0.4831 0.1094>, rot=<-0.0654 -0.0035 0.2242 -0.9723>}
 pinky_02_r {parent=pinky_01_r, loc=<-2.8274 -0.0893 0.1043>, rot=<0.0037 0.0011 0.2854 -0.9584>}
 pinky_03_r {parent=pinky_02_r, loc=<-2.2046 -0.0307 0.0078>, rot=<-0.0002 0.0098 0.1445 -0.9895>}
 deform_wrist_dn_base_r {parent=hand_r, loc=<-0.0001 1.5328 0>, rot=<0 0 0 -1>}
 deform_wrist_dn_r {parent=deform_wrist_dn_base_r, loc=<0 0 0>, rot=<0 0 0 1>}
 deform_wrist_up_base_r {parent=hand_r, loc=<-0.0002 -2.3007 -0.0001>, rot=<0 0 0 -1>}
 deform_wrist_up_r {parent=deform_wrist_up_base_r, loc=<0 0 0>, rot=<0 0 0 1>}
 lowerarm_twist_02_r {parent=lowerarm_r, loc=<-10.6707 0 0>, rot=<0 0 0 -1>}
 lowerarm_twist_01_r {parent=lowerarm_r, loc=<-18.316 0 0>, rot=<0 0 0 -1>}
 deform_elbow_in_base_r {parent=lowerarm_r, loc=<-0.1834 -3.6062 0.2572>, rot=<0 0 0 -1>}
 deform_elbow_in_r {parent=deform_elbow_in_base_r, loc=<0 0 0>, rot=<0 0 0 1>}
 deform_elbow_out_base_r {parent=lowerarm_r, loc=<0 4 -0.0001>, rot=<0 0 0 -1>}
 deform_elbow_out_r {parent=deform_elbow_out_base_r, loc=<0 0 0>, rot=<0 0 0 1>}
 upperarm_twist_01_r {parent=upperarm_r, loc=<-7.2135 0.1134 0>, rot=<0 0 0 -1>}
 deform_upperarm_bk_base_r {parent=upperarm_twist_01_r, loc=<6.4053 4.3746 -0.314>, rot=<0 0 0 -1>}
 deform_upperarm_bk_r {parent=deform_upperarm_bk_base_r, loc=<0 0 0>, rot=<0 0 0 1>}
 deform_upperarm_fr_base_r {parent=upperarm_twist_01_r, loc=<5.6312 -5.092 -0.3354>, rot=<0 0 0 -1>}
 deform_upperarm_fr_r {parent=deform_upperarm_fr_base_r, loc=<0 0 0>, rot=<0 0 0 1>}
 deform_upperarm_up_base_r {parent=upperarm_twist_01_r, loc=<8.891 -0.4825 -3.9517>, rot=<0 0 0 -1>}
 deform_upperarm_up_r {parent=deform_upperarm_up_base_r, loc=<0 0 0>, rot=<0 0 0 1>}
 upperarm_twist_02_r {parent=upperarm_r, loc=<-19.7998 0.1134 0>, rot=<0 0 0 -1>}
 dyn_shoulder_pad_st_a_r {parent=clavicle_r, loc=<-9.5142 1.4496 -3.6806>, rot=<0.0097 -0.3214 -0.0233 -0.9466>}
 dyn_shoulder_pad_st_b_r {parent=dyn_shoulder_pad_st_a_r, loc=<0 0 0>, rot=<0 0 0 -1>}
 head_null {parent=spine_05, type=Null, loc=<20.6573 3.5489 0>, rot=<0 0 -0.036 -0.9994>}
 deform_pec_aim_base_l {parent=spine_04, loc=<3.2456 9.3309 -1.2237>, rot=<0.0603 -0.4033 0.3318 -0.8507>}
 deform_pec_aim_l {parent=deform_pec_aim_base_l, loc=<0 0 0>, rot=<0 0 0 1>}
 deform_pec_base_l {parent=deform_pec_aim_l, loc=<9.3665 0 0>, rot=<0 0 0 1>}
 deform_pec_l {parent=deform_pec_base_l, loc=<0 0 0>, rot=<0 0 0 1>}
 deform_pec_aim_base_r {parent=spine_04, loc=<3.2452 9.3309 1.2237>, rot=<-0.4033 -0.0603 0.8507 0.3318>}
 deform_pec_aim_r {parent=deform_pec_aim_base_r, loc=<0 0 0>, rot=<0 0 0 1>}
 deform_pec_base_r {parent=deform_pec_aim_r, loc=<-9.3667 -0.0003 -0.0004>, rot=<0 0 0 1>}
 deform_pec_r {parent=deform_pec_base_r, loc=<0 0 0>, rot=<0 0 0 1>}
 deform_underarm_01_l {parent=spine_04, loc=<0.6418 1.2443 -9.9774>, rot=<0.0294 -0.7065 -0.0294 -0.7065>}
 deform_underarm_01_r {parent=spine_04, loc=<0.9501 -2.4469 9.9774>, rot=<0.7065 0.0294 -0.7065 0.0294>}
 upperarm_l_null {parent=pelvis, type=Null, loc=<37.2739 -6.6297 -13.0371>, rot=<-0.0145 -0.9125 0 -0.4088>}
 upperarm_r_null {parent=pelvis, type=Null, loc=<37.2739 -6.6297 13.0371>, rot=<0.9125 -0.0145 -0.4088 0>}
 thigh_l {parent=pelvis, loc=<-1.0855 -0.2109 -8.7053>, rot=<0.0009 0.0355 0.0155 -0.9993>}
 calf_l {parent=thigh_l, loc=<-42.5284 -0.0019 -0.7297>, rot=<0.0006 -0.0161 -0.0391 -0.9991>}
 foot_l {parent=calf_l, loc=<-46.7401 -2.5014 -0.592>, rot=<0.0427 -0.0238 0.0353 -0.9982>}
 ball_l {parent=foot_l, loc=<-3.5658 10.8066 -0.1544>, rot=<0.0001 -0.0001 -0.7063 -0.7079>}
 calf_twist_02_l {parent=calf_l, loc=<-19.1546 -1.7943 -0.2638>, rot=<0.0237 0.0095 -0.0035 -0.9997>}
 calf_twist_01_l {parent=calf_l, loc=<-35.8483 -1.9197 -0.5794>, rot=<0.0237 0.0095 -0.0035 -0.9997>}
 deform_knee_in_base_l {parent=calf_l, loc=<0.8031 -5.0231 0.2456>, rot=<0 0 0 -1>}
 deform_knee_in_l {parent=deform_knee_in_base_l, loc=<0 0 0>, rot=<0 0 0 1>}
 deform_knee_out_base_l {parent=calf_l, loc=<0 4 0>, rot=<0 0 0 -1>}
 deform_knee_out_l {parent=deform_knee_out_base_l, loc=<0 0 0>, rot=<0 0 0 1>}
 thigh_twist_01_l {parent=thigh_l, loc=<-27.952 0 0>, rot=<0.0013 0 0 -1>}
 thigh_r {parent=pelvis, loc=<-1.0855 -0.2109 8.7053>, rot=<0.0355 -0.0009 0.9993 0.0155>}
 calf_r {parent=thigh_r, loc=<42.5284 0.0019 0.7297>, rot=<0.0006 -0.0161 -0.0391 -0.9991>}
 foot_r {parent=calf_r, loc=<46.7401 2.5014 0.592>, rot=<0.0427 -0.0238 0.0353 -0.9982>}
 ball_r {parent=foot_r, loc=<3.5658 -10.8066 0.1544>, rot=<0.0001 -0.0001 -0.7063 -0.7079>}
 calf_twist_02_r {parent=calf_r, loc=<19.1546 1.7943 0.2638>, rot=<0.0237 0.0095 -0.0035 -0.9997>}
 calf_twist_01_r {parent=calf_r, loc=<35.8483 1.9197 0.5794>, rot=<0.0237 0.0095 -0.0035 -0.9997>}
 deform_knee_in_base_r {parent=calf_r, loc=<-0.8031 5.0231 -0.2457>, rot=<0 0 0 -1>}
 deform_knee_in_r {parent=deform_knee_in_base_r, loc=<0 0 0>, rot=<0 0 0 1>}
 deform_knee_out_base_r {parent=calf_r, loc=<0 -4 0>, rot=<0 0 0 -1>}
 deform_knee_out_r {parent=deform_knee_out_base_r, loc=<0 0 0>, rot=<0 0 0 1>}
 thigh_twist_01_r {parent=thigh_r, loc=<27.952 0 0>, rot=<0.0013 0 0 -1>}
 deform_groin_base_r {parent=thigh_r, loc=<-0.4801 -5.6776 0.2486>, rot=<0 0 0 -1>}
 deform_groin_r {parent=deform_groin_base_r, loc=<0 0 0>, rot=<0 0 0 1>}
 deform_glute_base_l {parent=pelvis, loc=<-7.3466 -9.0213 -9.1245>, rot=<0.0172 0.031 -0.4477 -0.8935>}
 deform_glute_l {parent=deform_glute_base_l, loc=<0 0 0>, rot=<0 0 0 1>}
 deform_glute_base_r {parent=pelvis, loc=<-7.3464 -9.0213 9.1245>, rot=<0.031 -0.0172 0.8935 -0.4477>}
 deform_glute_r {parent=deform_glute_base_r, loc=<0 0 0>, rot=<0 0 0 1>}
 deform_groin_base_l {parent=pelvis, loc=<-0.4132 5.4485 -8.9232>, rot=<0.0009 0.0355 0.0155 -0.9993>}
 deform_groin_l {parent=deform_groin_base_l, loc=<0 0 0>, rot=<0 0 0 1>}
 base_belt_c {parent=pelvis, loc=<1.9607 8.4135 0>, rot=<-0.5089 -0.491 -0.491 -0.5089>}
 dyn_belt_c {parent=base_belt_c, loc=<0 0 0>, rot=<0 0 0 1>}
 base_belt_area_l {parent=pelvis, loc=<-1.0855 -0.2109 -8.7053>, rot=<0.0009 0.0355 0.0155 -0.9993>}
 base_belt_lower_sd_l {parent=base_belt_area_l, loc=<-3.0773 0.4651 -8.0431>, rot=<-0.0019 -0.7317 0.0015 -0.6816>}
 dyn_belt_lower_sd_l {parent=base_belt_lower_sd_l, loc=<0 0 0>, rot=<0 0 0 1>}
 base_belt_lower_fr_l {parent=base_belt_area_l, loc=<-0.9716 7.1204 -0.7372>, rot=<-0.4803 -0.5545 -0.4416 -0.5165>}
 dyn_belt_lower_fr_l {parent=base_belt_lower_fr_l, loc=<0 0 0>, rot=<0 0 0 1>}
 base_belt_lower_bk_l {parent=base_belt_area_l, loc=<3.2967 -10.5968 0.1216>, rot=<0.5161 -0.5188 0.483 -0.4809>}
 dyn_belt_lower_bk_l {parent=base_belt_lower_bk_l, loc=<0 0 0>, rot=<0 0 0 1>}
 base_belt_area_r {parent=pelvis, loc=<-1.0855 -0.2109 8.7053>, rot=<0.0355 -0.0009 0.9993 0.0155>}
 base_belt_lower_fr_r {parent=base_belt_area_r, loc=<-0.1509 -6.3876 0.7139>, rot=<-0.4803 -0.5545 -0.4416 -0.5165>}
 dyn_belt_lower_fr_r {parent=base_belt_lower_fr_r, loc=<0 0 0>, rot=<0 0 0 1>}
 base_belt_lower_bk_r {parent=base_belt_area_r, loc=<-2.2517 8.9487 3.4596>, rot=<0.5161 -0.5188 0.483 -0.4809>}
 dyn_belt_lower_bk_r {parent=base_belt_lower_bk_r, loc=<0 0 0>, rot=<0 0 0 1>}
 base_belt_lower_sd_r {parent=base_belt_area_r, loc=<1.5309 -0.4575 7.736>, rot=<-0.0019 -0.7317 0.0015 -0.6816>}
 dyn_belt_lower_sd_r {parent=base_belt_lower_sd_r, loc=<0 0 0>, rot=<0 0 0 1>}
 base_pistol_r {parent=dyn_belt_lower_sd_r, loc=<-0.5369 1.79 -0.4704>, rot=<0 -1 0 0>}
 dyn_pistol_r {parent=base_pistol_r, loc=<0 0 0>, rot=<0 0 0 1>}
 head_ctrl {controlType=EULER_TRANSFORM, type=Control, loc=<0 0 0>, rot=<0 0 0 1>}
 clavicle_l_ctrl {controlType=EULER_TRANSFORM, type=Control, loc=<0 0 0>, rot=<0 0 0 1>}
 upperarm_l_ctrl {controlType=EULER_TRANSFORM, parent=clavicle_l_ctrl, type=Control, loc=<0 0 0>, rot=<0 0 0 1>}
 clavicle_r_ctrl {controlType=EULER_TRANSFORM, type=Control, loc=<0 0 0>, rot=<0 0 0 1>}
 upperarm_r_ctrl {controlType=EULER_TRANSFORM, parent=clavicle_r_ctrl, type=Control, loc=<0 0 0>, rot=<0 0 0 1>}
]
variables [
 CrouchBool {type=bool, default=False}
 Crouching {type=double, default=0}
 Floating {type=double, default=0}
 HeadRotationUp {type=double, default=0}
 HeadRotationUpCrouch {type=double, default=0}
 IsHeadUpDelta {type=double, default=0}
 NotSelectedPistol {type=double, default=0}
 NotTargeting {type=double, default=0}
 SelectedPickaxe {type=double, default=0}
 SelectedPistol {type=double, default=0}
 Skydiving {type=bool, default=False}
 SlectedWeapons {type=double, default=0}
 Standing {type=double, default=0}
 TacticalSprinting {type=double, default=0}
 Targeting {type=double, default=0}
 UpperARmUpTargetingMode {type=double, default=0}
 UpperArmUpL {type=double, default=0}
 UpperArmUpR {type=double, default=0}
 Weight {type=float, default=0}
]
