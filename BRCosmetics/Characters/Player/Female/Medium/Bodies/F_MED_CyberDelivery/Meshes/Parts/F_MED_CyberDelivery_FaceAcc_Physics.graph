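# Physics asset for the CyberDelivery female medium FaceAcc mesh part, setting up capsule bodies and joints for dynamic side and back hair bones.
# Four hair-area bodies hang off the head with tight swing limits and carry the strands, some through free-rotating base_hair bodies; the center-back strand's base joins the head directly.
# Strand joints limit swing and mostly lock twist.
# Head, neck, upper spine, upper arms and pelvis form a free-rotating body skeleton.
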PhysicsAsset /BRCosmetics/Characters/Player/Female/Medium/Bodies/F_MED_CyberDelivery/Meshes/Parts/F_MED_CyberDelivery_FaceAcc_Physics
bodies {
 base_hair_bk_a_l [{shapeType=Capsule, length=0.505, radius=0.505, shapeName=base_hair_bk_a_l_capsule, rotation=<rot 0 -90 0>}]
 base_hair_bk_area_c [{shapeType=Capsule, length=0.505, radius=0.505, shapeName=base_hair_bk_area_c_capsule, rotation=<rot 0 -90 0>}]
 base_hair_bk_c_l [{shapeType=Capsule, length=0.505, radius=0.505, shapeName=base_hair_bk_c_l_capsule, rotation=<rot 0 -90 0>}]
 base_hair_bk_d_l [{shapeType=Capsule, length=0.505, radius=0.505, shapeName=base_hair_bk_d_l_capsule, rotation=<rot 0 -90 0>}]
 base_hair_bk_d_r [{shapeType=Capsule, length=0.505, radius=0.505, shapeName=base_hair_bk_d_r_capsule, rotation=<rot 0 -90 0>}]
 base_hair_sd_c_l [{shapeType=Capsule, length=0.505, radius=0.505, shapeName=base_hair_sd_c_l_capsule, rotation=<rot 0 -90 0>}]
 base_hair_sd_c_r [{shapeType=Capsule, length=0.505, radius=0.505, shapeName=base_hair_sd_c_r_capsule, rotation=<rot 0 -90 0>}]
 dyn_hair_area_bk_l [{shapeType=Capsule, length=1, radius=1, shapeName=dyn_hair_area_bk_l_capsule}]
 dyn_hair_area_bk_r [{
   shapeType=Capsule
   length=1
   radius=1
   shapeName=dyn_hair_area_bk_r_capsule
   center=<5.6084 0.0644 0.1894>
   rotation=<rot 24.951 -90 12.2991>
  }]
 dyn_hair_area_sd_l [{shapeType=Capsule, length=0.5, radius=0.5, shapeName=dyn_hair_area_sd_l_capsule}]
 dyn_hair_area_sd_r [{shapeType=Capsule, length=1, radius=1, shapeName=dyn_hair_area_sd_r_capsule}]
 dyn_hair_bk_a_c [{shapeType=Capsule, length=2, radius=1, shapeName=dyn_hair_bk_a_c_capsule, center=<1.0199 0 0>, rotation=<rot 90 0 0>}]
 dyn_hair_bk_a_l [{
   shapeType=Capsule
   length=3.5
   radius=1.5
   shapeName=dyn_hair_bk_a_l_capsule
   center=<1.8576 0.13 0.8>
   rotation=<rot 86.4197 -179.9835 -179.9845>
  }]
 dyn_hair_bk_a_r [{shapeType=Capsule, length=3, radius=2, shapeName=dyn_hair_bk_a_r_capsule, center=<2.6671 0 0.64>, rotation=<rot 71.2843 179.9999 179.9999>}]
 dyn_hair_bk_b_l [{shapeType=Capsule, length=3, radius=2, shapeName=dyn_hair_bk_b_l_capsule, center=<1.138 0 0.3>, rotation=<rot 90 0 0>}]
 dyn_hair_bk_b_r [{shapeType=Capsule, length=2, radius=1, shapeName=dyn_hair_bk_b_r_capsule, center=<3.1851 0 0.3>, rotation=<rot 90 0 0>}]
 dyn_hair_bk_c_l [{shapeType=Capsule, length=3.4, radius=1.5, shapeName=dyn_hair_bk_c_l_capsule, center=<2.5865 0 0>, rotation=<rot 90 0 0>}]
 dyn_hair_bk_c_r [{shapeType=Capsule, length=1.4, radius=1.1, shapeName=dyn_hair_bk_c_r_capsule, center=<1.5847 0 0>, rotation=<rot 90 0 0>}]
 dyn_hair_bk_d_l [{shapeType=Capsule, length=2.3951, radius=1.6226, shapeName=dyn_hair_bk_d_l_capsule, center=<0.8769 0 0>, rotation=<rot 90 0 0>}]
 dyn_hair_bk_d_r [{shapeType=Capsule, length=1.2, radius=1, shapeName=dyn_hair_bk_d_r_capsule, center=<2.194 0 0>, rotation=<rot 90 0 0>}]
 dyn_hair_bk_e_l [{shapeType=Capsule, length=1.1471, radius=0.8654, shapeName=dyn_hair_bk_e_l_capsule, center=<1.8548 0 0>, rotation=<rot 90 0 0>}]
 dyn_hair_bk_f_l [{shapeType=Capsule, length=3, radius=1, shapeName=dyn_hair_bk_f_l_capsule, center=<0.4 0 0>, rotation=<rot 90 0 0>}]
 dyn_hair_sd_a_l [{shapeType=Capsule, length=2, radius=1, shapeName=dyn_hair_sd_a_l_capsule, center=<1.0383 0 0>, rotation=<rot 90 0 0>}]
 dyn_hair_sd_a_r [{shapeType=Capsule, length=2, radius=1, shapeName=dyn_hair_sd_a_r_capsule, center=<1.1936 0 0>, rotation=<rot 90 0 0>}]
 dyn_hair_sd_b_l [{shapeType=Capsule, length=2, radius=1.5, shapeName=dyn_hair_sd_b_l_capsule, center=<1.6364 0 0>, rotation=<rot 90 0 0>}]
 dyn_hair_sd_b_r [{shapeType=Capsule, length=3.3046, radius=1.5, shapeName=dyn_hair_sd_b_r_capsule, center=<0.6882 0 0>, rotation=<rot 90 0 0>}]
 dyn_hair_sd_c_l [{shapeType=Capsule, length=1.5, radius=1, shapeName=dyn_hair_sd_c_l_capsule, center=<1.5435 0 0>, rotation=<rot 90 0 0>}]
 dyn_hair_sd_c_r [{shapeType=Capsule, length=3, radius=1.5, shapeName=dyn_hair_sd_c_r_capsule, center=<2.8393 0 0.66>, rotation=<rot 80.3473 179.8196 179.8348>}]
 dyn_hair_sd_d_l [{shapeType=Capsule, length=2.1, radius=1, shapeName=dyn_hair_sd_d_l_capsule, center=<2.6689 0 0.2>, rotation=<rot 0 90.9849 -85.1899>}]
 dyn_hair_sd_d_r [{shapeType=Capsule, length=1.8, radius=1, shapeName=dyn_hair_sd_d_r_capsule, center=<1.0244 0 0>, rotation=<rot 90 0 0>}]
 dyn_hair_sd_e_l [{
   shapeType=Capsule
   length=3.2339
   radius=1.1771
   shapeName=dyn_hair_sd_e_l_capsule
   center=<3.1754 0.0557 -0.0414>
   rotation=<rot 0 -87.3925 -83.4136>
  }]
 dyn_hair_sd_e_r [{shapeType=Capsule, length=3.5, radius=1.3, shapeName=dyn_hair_sd_e_r_capsule, center=<3.3522 0 0>, rotation=<rot 90 0 0>}]
 dyn_hair_sd_f_l [{shapeType=Capsule, length=1.5, radius=1.5, shapeName=dyn_hair_sd_f_l_capsule, center=<2.2467 0 0>, rotation=<rot 90 0 0>}]
 dyn_hair_sd_f_r [{shapeType=Capsule, length=3, radius=1.2, shapeName=dyn_hair_sd_f_r_capsule, center=<2.5475 0 0>, rotation=<rot 90 0 0>}]
 dyn_hair_sd_g_l [{shapeType=Capsule, length=1.9, radius=1, shapeName=dyn_hair_sd_g_l_capsule, center=<1.9804 0 0>, rotation=<rot 90 0 0>}]
 dyn_hair_sd_g_r [{shapeType=Capsule, length=1.5, radius=1, shapeName=dyn_hair_sd_g_r_capsule, center=<1.8376 0.12 0>, rotation=<rot 90 0 0>}]
 dyn_hair_sd_h_l [{shapeType=Capsule, length=1, radius=0.5, shapeName=dyn_hair_sd_h_l_capsule, center=<2.4063 0 0>, rotation=<rot 90 0 0>}]
 dyn_hair_sd_h_r [{shapeType=Capsule, length=1.5, radius=0.8, shapeName=dyn_hair_sd_h_r_capsule, center=<1.7742 0 -0.7>, rotation=<rot 90 0 0>}]
 head [{shapeType=Capsule, length=2, radius=2, shapeName=head_capsule, center=<5.1481 -2.2951 -0.2356>, rotation=<rot 40.6084 -90 -91.405>}]
 neck_01 [{shapeType=Capsule, length=11.505, radius=8, shapeName=neck_01_capsule, rotation=<rot 90 0 0>}]
 pelvis [{shapeType=Capsule, length=0.505, radius=0.505, shapeName=pelvis_capsule, rotation=<rot 0 -90 0>}]
 spine_05 [
  {shapeType=Capsule, length=17, radius=10, shapeName=spine_05_capsule, center=<1.321 -3.8453 0>}
  {shapeType=Capsule, length=7.9042, radius=4.4244, shapeName=spine_05_capsule, center=<3.8286 -9.4255 0>, rotation=<rot 90 180 155.6369>}
 ]
 upperarm_l [{shapeType=Capsule, length=20, radius=6, shapeName=upperarm_r_capsule, center=<9.9314 0 0>, rotation=<rot 90 0 180>}]
 upperarm_r [{shapeType=Capsule, length=20, radius=6, shapeName=upperarm_r_capsule, center=<-9.9314 0 0>, rotation=<rot 90 0 0>}]
}
constraints [
 {
  bone1Name=dyn_hair_sd_b_r
  bone2Name=dyn_hair_area_sd_r
  swing1LimitDegrees=20
  swing2LimitDegrees=15
  twistLimitDegrees=45
  twistMotion=Locked
 }
 {
  bone1Name=dyn_hair_sd_c_r
  bone2Name=base_hair_sd_c_r
  swing1LimitDegrees=20
  swing2LimitDegrees=20
  twistLimitDegrees=45
  twistMotion=Locked
 }
 {
  bone1Name=dyn_hair_sd_a_r
  bone2Name=dyn_hair_area_sd_r
  swing1LimitDegrees=20
  swing2LimitDegrees=15
  twistLimitDegrees=45
  twistMotion=Locked
 }
 {
  bone1Name=dyn_hair_sd_a_l
  bone2Name=dyn_hair_area_sd_l
  swing1LimitDegrees=20
  swing2LimitDegrees=15
  twistLimitDegrees=45
  twistMotion=Locked
 }
 {
  bone1Name=dyn_hair_sd_b_l
  bone2Name=dyn_hair_area_sd_l
  swing1LimitDegrees=20
  swing2LimitDegrees=15
  twistLimitDegrees=45
  twistMotion=Locked
 }
 {
  bone1Name=dyn_hair_bk_a_r
  bone2Name=dyn_hair_area_bk_r
  swing1LimitDegrees=20
  swing2LimitDegrees=35
  twistLimitDegrees=45
  twistMotion=Locked
 }
 {
  bone1Name=dyn_hair_bk_b_r
  bone2Name=dyn_hair_area_bk_r
  swing1LimitDegrees=20
  swing2LimitDegrees=35
  twistLimitDegrees=45
  twistMotion=Locked
 }
 {
  bone1Name=dyn_hair_bk_a_l
  bone2Name=base_hair_bk_a_l
  swing1LimitDegrees=20
  swing2LimitDegrees=30
  twistLimitDegrees=45
  twistMotion=Locked
 }
 {
  bone1Name=dyn_hair_bk_b_l
  bone2Name=dyn_hair_area_bk_l
  swing1LimitDegrees=20
  swing2LimitDegrees=30
  twistLimitDegrees=45
  twistMotion=Locked
 }
 {
  bone1Name=dyn_hair_bk_c_l
  bone2Name=base_hair_bk_c_l
  swing1LimitDegrees=15
  swing2LimitDegrees=30
  twistLimitDegrees=45
  twistMotion=Locked
 }
 {bone1Name=dyn_hair_area_sd_r, bone2Name=head, swing1LimitDegrees=5, swing2LimitDegrees=5, twistLimitDegrees=45, twistMotion=Locked}
 {bone1Name=dyn_hair_area_bk_l, bone2Name=head, swing1LimitDegrees=5, swing2LimitDegrees=5, twistLimitDegrees=45, twistMotion=Locked}
 {bone1Name=dyn_hair_area_sd_l, bone2Name=head, swing1LimitDegrees=5, swing2LimitDegrees=5, twistLimitDegrees=45, twistMotion=Locked}
 {bone1Name=dyn_hair_area_bk_r, bone2Name=head, swing1LimitDegrees=5, swing2LimitDegrees=5, twistLimitDegrees=45, twistMotion=Locked}
 {
  bone1Name=head
  bone2Name=neck_01
  swing1LimitDegrees=45
  swing1Motion=Free
  swing2LimitDegrees=45
  swing2Motion=Free
  twistLimitDegrees=45
  twistMotion=Free
 }
 {
  bone1Name=upperarm_l
  bone2Name=spine_05
  swing1LimitDegrees=45
  swing1Motion=Free
  swing2LimitDegrees=45
  swing2Motion=Free
  twistLimitDegrees=45
  twistMotion=Free
 }
 {
  bone1Name=upperarm_r
  bone2Name=spine_05
  swing1LimitDegrees=45
  swing1Motion=Free
  swing2LimitDegrees=45
  swing2Motion=Free
  twistLimitDegrees=45
  twistMotion=Free
 }
 {
  bone1Name=spine_05
  bone2Name=pelvis
  swing1LimitDegrees=45
  swing1Motion=Free
  swing2LimitDegrees=45
  swing2Motion=Free
  twistLimitDegrees=45
  twistMotion=Free
 }
 {
  bone1Name=neck_01
  bone2Name=spine_05
  swing1LimitDegrees=45
  swing1Motion=Free
  swing2LimitDegrees=45
  swing2Motion=Free
  twistLimitDegrees=45
  twistMotion=Free
 }
 {
  bone1Name=dyn_hair_bk_a_c
  bone2Name=base_hair_bk_area_c
  swing1LimitDegrees=20
  swing2LimitDegrees=20
  twistLimitDegrees=45
  twistMotion=Locked
 }
 {
  bone1Name=base_hair_bk_area_c
  bone2Name=head
  swing1LimitDegrees=45
  swing1Motion=Free
  swing2LimitDegrees=45
  swing2Motion=Free
  twistLimitDegrees=45
  twistMotion=Free
 }
 {
  bone1Name=dyn_hair_sd_d_r
  bone2Name=dyn_hair_area_sd_r
  swing1LimitDegrees=20
  swing2LimitDegrees=15
  twistLimitDegrees=45
  twistMotion=Locked
 }
 {
  bone1Name=dyn_hair_sd_e_r
  bone2Name=dyn_hair_area_sd_r
  swing1LimitDegrees=20
  swing2LimitDegrees=15
  twistLimitDegrees=45
  twistMotion=Locked
 }
 {
  bone1Name=dyn_hair_sd_f_r
  bone2Name=dyn_hair_area_sd_r
  swing1LimitDegrees=20
  swing2LimitDegrees=15
  twistLimitDegrees=45
  twistMotion=Locked
 }
 {
  bone1Name=dyn_hair_sd_g_r
  bone2Name=dyn_hair_area_sd_r
  swing1LimitDegrees=20
  swing2LimitDegrees=15
  twistLimitDegrees=45
  twistMotion=Locked
 }
 {
  bone1Name=dyn_hair_sd_h_r
  bone2Name=dyn_hair_area_sd_r
  swing1LimitDegrees=20
  swing2LimitDegrees=15
  twistLimitDegrees=45
  twistMotion=Locked
 }
 {
  bone1Name=dyn_hair_sd_c_l
  bone2Name=base_hair_sd_c_l
  swing1LimitDegrees=20
  swing2LimitDegrees=15
  twistLimitDegrees=45
  twistMotion=Locked
 }
 {
  bone1Name=dyn_hair_sd_d_l
  bone2Name=dyn_hair_area_sd_l
  swing1LimitDegrees=20
  swing2LimitDegrees=15
  twistLimitDegrees=45
  twistMotion=Locked
 }
 {
  bone1Name=dyn_hair_sd_e_l
  bone2Name=dyn_hair_area_sd_l
  swing1LimitDegrees=25
  swing2LimitDegrees=25
  twistLimitDegrees=45
  twistMotion=Locked
 }
 {
  bone1Name=dyn_hair_sd_f_l
  bone2Name=dyn_hair_area_sd_l
  swing1LimitDegrees=20
  swing2LimitDegrees=15
  twistLimitDegrees=45
  twistMotion=Locked
 }
 {
  bone1Name=dyn_hair_bk_c_r
  bone2Name=dyn_hair_area_bk_r
  swing1LimitDegrees=20
  swing2LimitDegrees=35
  twistLimitDegrees=45
  twistMotion=Locked
 }
 {bone1Name=dyn_hair_bk_d_r, bone2Name=base_hair_bk_d_r, swing1LimitDegrees=20, swing2LimitDegrees=25, twistLimitDegrees=2}
 {
  bone1Name=dyn_hair_bk_d_l
  bone2Name=base_hair_bk_d_l
  swing1LimitDegrees=15
  swing2LimitDegrees=30
  twistLimitDegrees=45
  twistMotion=Locked
 }
 {
  bone1Name=dyn_hair_bk_f_l
  bone2Name=dyn_hair_area_bk_l
  swing1LimitDegrees=10
  swing2LimitDegrees=25
  twistLimitDegrees=45
  twistMotion=Locked
 }
 {
  bone1Name=dyn_hair_bk_e_l
  bone2Name=dyn_hair_area_bk_l
  swing1LimitDegrees=20
  swing2LimitDegrees=25
  twistLimitDegrees=45
  twistMotion=Locked
 }
 {
  bone1Name=dyn_hair_sd_h_l
  bone2Name=dyn_hair_area_sd_l
  swing1LimitDegrees=20
  swing2LimitDegrees=15
  twistLimitDegrees=45
  twistMotion=Locked
 }
 {
  bone1Name=dyn_hair_sd_g_l
  bone2Name=dyn_hair_area_sd_l
  swing1LimitDegrees=20
  swing2LimitDegrees=15
  twistLimitDegrees=45
  twistMotion=Locked
 }
 {
  bone1Name=base_hair_bk_a_l
  bone2Name=dyn_hair_area_bk_l
  swing1LimitDegrees=45
  swing1Motion=Free
  swing2LimitDegrees=45
  swing2Motion=Free
  twistLimitDegrees=45
  twistMotion=Free
 }
 {
  bone1Name=base_hair_bk_c_l
  bone2Name=dyn_hair_area_bk_l
  swing1LimitDegrees=45
  swing1Motion=Free
  swing2LimitDegrees=45
  swing2Motion=Free
  twistLimitDegrees=45
  twistMotion=Free
 }
 {
  bone1Name=base_hair_bk_d_l
  bone2Name=dyn_hair_area_bk_l
  swing1LimitDegrees=45
  swing1Motion=Free
  swing2LimitDegrees=45
  swing2Motion=Free
  twistLimitDegrees=45
  twistMotion=Free
 }
 {
  bone1Name=base_hair_bk_d_r
  bone2Name=dyn_hair_area_bk_r
  swing1LimitDegrees=45
  swing1Motion=Free
  swing2LimitDegrees=45
  swing2Motion=Free
  twistLimitDegrees=45
  twistMotion=Free
 }
 {
  bone1Name=base_hair_sd_c_r
  bone2Name=dyn_hair_area_sd_r
  swing1LimitDegrees=45
  swing1Motion=Free
  swing2LimitDegrees=45
  swing2Motion=Free
  twistLimitDegrees=45
  twistMotion=Free
 }
 {
  bone1Name=base_hair_sd_c_l
  bone2Name=dyn_hair_area_sd_l
  swing1LimitDegrees=45
  swing1Motion=Free
  swing2LimitDegrees=45
  swing2Motion=Free
  twistLimitDegrees=45
  twistMotion=Free
 }
]
