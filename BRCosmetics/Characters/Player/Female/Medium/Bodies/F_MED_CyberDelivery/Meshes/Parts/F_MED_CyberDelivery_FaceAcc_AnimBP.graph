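# Post-process style AnimBP for the CyberDelivery female face accessory, driving its hair bones to follow and react to the player's body.
# Copied parent and body-mesh poses feed a Control Rig driven by movement, aim and weapon state.
# Two head-space RigidBody sims, blended on crouch, add secondary motion before additive bone offsets adjust the hair.
# The event graph updates those state alphas and a state-based gravity override for the sims.
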AnimBlueprint /BRCosmetics/Characters/Player/Female/Medium/Bodies/F_MED_CyberDelivery/Meshes/Parts/F_MED_CyberDelivery_FaceAcc_AnimBP
target_skeleton=&/F_MED_CyberDelivery_FaceAcc_Skeleton
graphs [
 AnimGraph {nodes [
   n1 = RigidBody!(
    ComponentPose=Local To Component(LocalPose=Use cached pose 'CtrlRigCache'())
    Alpha
    OverrideWorldGravity:Vector=Get GravityOverride()
    ExternalForce
   ) -> (Pose) {
    bOverrideWorldGravity=true
    overridePhysicsAsset=&/F_MED_CyberDelivery_FaceAcc_Physics
    simulationSpace=BaseBoneSpace
    alphaBoolBlend {blendOption=Linear}
    baseBoneRef {boneName=head}
    componentAppliedLinearAccClamp=<3000 3000 3000>
    componentLinearAccScale=<0.4 0.3 0.3>
    componentLinearVelScale=<0.4 0.3 0.3>
    overrideWorldGravity {z=-980}
    simSpaceSettings {
     maxAngularAcceleration=10
     maxAngularVelocity=3
     maxLinearAcceleration=1000
     maxLinearVelocity=500
     worldAlpha=1
     externalLinearDragV=<1 1 1>
    }
   }
   n2 = Component To Local(ComponentPose=@n1)
   n3 = RigidBody(ComponentPose=Local To Component(LocalPose=Use cached pose 'CtrlRigCache'()), OverrideWorldGravity=Get GravityOverride()) {
    bOverrideWorldGravity=true
    overridePhysicsAsset=&F_MED_CyberDelivery_FaceAcc_Physics
    simulationSpace=BaseBoneSpace
    alphaBoolBlend {blendOption=Linear}
    baseBoneRef {boneName=head}
    componentAppliedLinearAccClamp=<3000 3000 3000>
    componentLinearAccScale=<0.3 0.3 0.3>
    componentLinearVelScale=<0.3 0.3 0.3>
    overrideWorldGravity {z=-980}
    simSpaceSettings {
     maxAngularAcceleration=2
     maxAngularVelocity=3
     maxLinearAcceleration=1000
     maxLinearVelocity=1000
     velocityScaleZ=0.3
     worldAlpha=1
     externalLinearDragV=<0.2 0.2 0.2>
    }
   }
   n4 = Component To Local(ComponentPose=@n3)
   n5 = Get bIsCrouching()
   n6 = "Transform (Modify) Bone - Bone: base_hair_area_sd_l"(
    ComponentPose=Local To Component(LocalPose=Blend Poses by bool(BlendPose_0=@n2, BlendPose_1=@n4, bActiveValue=@n5) {blendType=Linear})
    Alpha=1
    Translation="-1.282,0.000,-2.034"
    Rotation="-11.430,-22.429,4.676"
    Scale="1,1,1"
   ) {
    rotationMode=BMM_Additive
    rotationSpace=BCS_BoneSpace
    scaleMode=BMM_Additive
    scaleSpace=BCS_BoneSpace
    translationMode=BMM_Additive
    translationSpace=BCS_BoneSpace
    alphaBoolBlend {blendOption=Linear}
    alphaScaleBiasClamp {bInterpResult=true, interpSpeedIncreasing=3}
    boneToModify {boneName=base_hair_area_sd_l}
    rotation=<rot -11.43 -22.429 4.676>
    translation {x=-1.282, z=-2.034}
   }
   n7 = "Transform (Modify) Bone - Bone: base_hair_area_bk_l"(ComponentPose=@n6, Alpha=1, Translation="-0.917,-1.716,1.253", Rotation="-0.000,-5.292,0.000", Scale="1,1,1") {
    rotationMode=BMM_Additive
    rotationSpace=BCS_BoneSpace
    scaleMode=BMM_Additive
    scaleSpace=BCS_BoneSpace
    translationMode=BMM_Additive
    translationSpace=BCS_BoneSpace
    alphaBoolBlend {blendOption=Linear}
    alphaScaleBiasClamp {bInterpResult=true, interpSpeedIncreasing=3}
    boneToModify {boneName=base_hair_area_bk_l}
    rotation=<rot 0 -5.292 0>
    translation=<-0.917 -1.7159 1.253>
   }
   n8 = "Transform (Modify) Bone - Bone: base_hair_area_sd_r"(ComponentPose=@n7, Alpha=1, Translation="-0.543,1.391,-0.920", Rotation="-18.471,9.322,0.190", Scale="1,1,1") {
    rotationMode=BMM_Additive
    rotationSpace=BCS_BoneSpace
    scaleMode=BMM_Additive
    scaleSpace=BCS_BoneSpace
    translationMode=BMM_Additive
    translationSpace=BCS_BoneSpace
    alphaBoolBlend {blendOption=Linear}
    alphaScaleBiasClamp {bInterpResult=true, interpSpeedIncreasing=3}
    boneToModify {boneName=base_hair_area_sd_r}
    rotation=<rot -18.471 9.322 0.19>
    translation=<-0.543 1.391 -0.92>
   }
   n9 = "Transform (Modify) Bone - Bone: base_hair_area_bk_r"(ComponentPose=@n8, Alpha=1, Translation="-0.592,0.408,1.298", Rotation="0,0,0", Scale="1,1,1") {
    rotationMode=BMM_Additive
    rotationSpace=BCS_BoneSpace
    scaleMode=BMM_Additive
    scaleSpace=BCS_BoneSpace
    translationMode=BMM_Additive
    translationSpace=BCS_BoneSpace
    alphaBoolBlend {blendOption=Linear}
    alphaScaleBiasClamp {bInterpResult=true, interpSpeedIncreasing=3}
    boneToModify {boneName=base_hair_area_bk_r}
    translation=<-0.592 0.408 1.298>
   }
   n10 = "Transform (Modify) Bone - Bone: base_hair_bk_a_l"(ComponentPose=@n9, Alpha=1, Translation="-0.284,-0.110,-0.341", Rotation="-64.953,44.344,-43.130", Scale="0.7,0.7,0.7") {
    rotationMode=BMM_Additive
    rotationSpace=BCS_BoneSpace
    scaleMode=BMM_Additive
    scaleSpace=BCS_BoneSpace
    translationMode=BMM_Additive
    translationSpace=BCS_BoneSpace
    alphaBoolBlend {blendOption=Linear}
    alphaScaleBiasClamp {bInterpResult=true, interpSpeedIncreasing=3}
    boneToModify {boneName=base_hair_bk_a_l}
    rotation=<rot -64.953 44.3441 -43.13>
    scale=<0.7 0.7 0.7>
    translation=<-0.284 -0.11 -0.341>
   }
   n11 = "Transform (Modify) Bone - Bone: base_hair_area_bk_l"(ComponentPose=@n10, Alpha=1, Translation="2.020,-0.401,-2.349", Rotation="-8.488,-1.017,0.621", Scale="1,1,1") {
    rotationMode=BMM_Additive
    rotationSpace=BCS_BoneSpace
    scaleMode=BMM_Additive
    scaleSpace=BCS_BoneSpace
    translationMode=BMM_Additive
    translationSpace=BCS_BoneSpace
    alphaBoolBlend {blendOption=Linear}
    alphaScaleBiasClamp {bInterpResult=true, interpSpeedIncreasing=2}
    boneToModify {boneName=base_hair_area_bk_l}
    rotation=<rot -8.488 -1.017 0.621>
    translation=<2.0199 -0.401 -2.3487>
   }
   n12 = "Transform (Modify) Bone - Bone: base_hair_area_bk_r"(ComponentPose=@n11, Alpha=1, Translation="2.895,-0.610,2.210", Rotation="0,0,0", Scale="1,1,1") {
    rotationMode=BMM_Additive
    rotationSpace=BCS_BoneSpace
    scaleMode=BMM_Additive
    scaleSpace=BCS_BoneSpace
    translationMode=BMM_Additive
    translationSpace=BCS_BoneSpace
    alphaBoolBlend {blendOption=Linear}
    alphaScaleBiasClamp {bInterpResult=true, interpSpeedIncreasing=2}
    boneToModify {boneName=base_hair_area_bk_r}
    translation=<2.8953 -0.61 2.2103>
   }
   Output Pose(Result=Component To Local(ComponentPose=@n12)) {layerGroup=None}
   n13 = Layered blend per bone(
    BasePose=Copy Pose From Mesh() {bUseAttachedParent=true}
    BlendPoses_0=Copy Pose From Mesh(SourceMeshComponent=Get Body_Mesh()) {bUseAttachedParent=true}
   ) {layerSetup [{branchFilters [{boneName=clavicle_l}, {boneName=clavicle_r}, {boneName=upperarm_r}, {boneName=upperarm_l}]}]}
   Save cached pose 'StartCache'(Pose=@n13) {cacheName=StartCache}
   n14 = Control Rig!(
    Source=Use cached pose 'StartCache'()
    Alpha
    Targeting:double=Get Alpha_Targeting() -> (Alpha_Targeting:double)
    TacticalSprinting:double=Get Alpha_Tactical_Sprinting() -> (Alpha_Tactical_Sprinting:double)
    Floating:double=Get Alpha_Floating() -> (Alpha_Floating:double)
    SelectedPistol:double=Get Selected_Pistol() -> (Selected_Pistol:double)
    Skydiving:bool=Get bIsSkyDiving()
    Crouching:double=Get Alpha_Crouching() -> (Alpha_Crouching:double)
    NotTargeting:double=Get Alpha_NotTargeting() -> (Alpha_NotTargeting:double)
    Standing:double=Get Alpha_Standing() -> (Alpha_Standing:double)
    SelectedPickaxe:double=Get Selected_Pickaxe() -> (Selected_Pickaxe:double)
    NotSelectedPistol:double=Get NotSelected_Pistol() -> (NotSelected_Pistol:double)
    CrouchBool:bool=Get bIsCrouching()
   ) -> (Pose) {
    alphaBoolBlend {blendOption=Linear}
    controlRigAssetReference {blueprintRigClass=&/F_MED_CyberDelivery_FaceAcc_CtrlRig.F_MED_CyberDelivery_FaceAcc_CtrlRig_C}
   }
   Save cached pose 'CtrlRigCache'(Pose=@n14) {cacheName=CtrlRigCache}
  ]}
 EventGraph {nodes [
   n1 = Event BlueprintUpdateAnimation()
   > Set Delta_TimeX(Delta_TimeX:double=@n1.DeltaTimeX) -> (Output_Get:double)
   > n2 = Sequence!(execute) -> (
    then_0
    then_1
    then_2:Exec
    then_3:Exec
    then_4:Exec
    then_5:Exec
    then_6:Exec
    then_7:Exec
    then_8:Exec
    then_9:Exec
    then_10:Exec
    then_11:Exec
   )
   n3 = Gravity Override State Function DEPRECATED(
    execute=@n2.then_1
    CharacterPartAnimInstance=Self-Reference()
    GameTimeInSeconds=GetGameTimeInSeconds()
    GravityOverrideParams=Get Gravity Override Params()
    RBAN Params=Get RBAN Params()
   )
   > Set GravityOverride(GravityOverride=@n3)
   > Set Sim Space Settings(Sim Space Settings:Sim Space Settings Structure=Break RBANSetting Struct!(RBANSettingStruct:"RBANSetting Struct Structure (by ref)"=@n3.RBAN SettingsOut) -> (
     ComponentLinearAccScale_28_#1:Vector=~
     ComponentLinearVelScale_31_#2:Vector=~
     SimSpaceSettings_34_#3:Sim Space Settings Structure
    ) {is=title K2Node_BreakStruct}.SimSpaceSettings_34_#3) -> (Output_Get:Sim Space Settings Structure)
   n4 = Get bIsSliding()
   n5 = Branch(
    execute=@n2.then_9
    Condition=AND Boolean!(A=Get bIsCrouching(), B=Get bIsMoving2D(), C:bool=Get bIsTargeting(), D:bool=NOT Boolean(A=@n4)) -> (ReturnValue)
   )
   > Set Alpha_CrouchMoving2DTargeting(Alpha_CrouchMoving2DTargeting:double=1) -> (Output_Get:double)
   Set Alpha_CrouchMoving2DTargeting(execute=@n5.else, Alpha_CrouchMoving2DTargeting=0)
   n6 = GetAnimInstance(self=GetOwnerSkeletalMeshComponent())
   Set PlayingCelebration(
    execute=@n2.then_11
    PlayingCelebration:double="To Float (Boolean)"(InBool=Montage_IsPlaying(self=@n6, Montage=/BRCosmetics/Animation/Game/MainPlayer/Emotes/Emote_Celebration_Female_M))
   ) -> (Output_Get:double)
   Reroute Node(InputPin=@n2.then_0)
   n7 = Get bIsTargeting()
   Set Alpha_Targeting(
    execute=@n2.then_2
    Alpha_Targeting:double=FInterpTo(Current=Get Alpha_Targeting(), Target="To Float (Boolean)"(InBool=@n7), DeltaTime=Get Delta_TimeX() -> (Delta_TimeX:double), InterpSpeed=10)
   ) -> (Output_Get:double)
   n8 = Get bIsSkydiveFloating()
   Set Alpha_Floating(
    execute=@n2.then_4
    Alpha_Floating:double=FInterpTo(Current=Get Alpha_Floating(), Target="To Float (Boolean)"(InBool=@n8), DeltaTime=Get Delta_TimeX(), InterpSpeed=5)
   ) -> (Output_Get:double)
   n9 = AND Boolean!(A=Get bIsCrouching(), B=NOT Boolean(A=Get bIsSliding())) -> (ReturnValue)
   Set Alpha_Crouching(
    execute=@n2.then_5
    Alpha_Crouching:double=FInterpTo(Current=Get Alpha_Crouching(), Target="To Float (Boolean)"(InBool=@n9), DeltaTime=Get Delta_TimeX(), InterpSpeed=10)
   ) -> (Output_Get:double)
   n10 = Get bIsTacticalSprinting()
   Set Alpha_Tactical_Sprinting(
    execute=@n2.then_3
    Alpha_Tactical_Sprinting:double=FInterpTo(Current=Get Alpha_Tactical_Sprinting(), Target="To Float (Boolean)"(InBool=@n10), DeltaTime=Get Delta_TimeX(), InterpSpeed=10)
   ) -> (Output_Get:double)
   n11 = NOT Boolean(A=Get bIsTargeting())
   Set Alpha_NotTargeting(
    execute=@n2.then_6
    Alpha_NotTargeting:double=FInterpTo(Current=Get Alpha_NotTargeting(), Target="To Float (Boolean)"(InBool=@n11), DeltaTime=Get Delta_TimeX(), InterpSpeed=20)
   ) -> (Output_Get:double)
   n12 = Get bIsSliding()
   n13 = Branch(execute=@n2.then_7, Condition=AND Boolean!(A=Get bIsCrouching(), B=Get bIsMoving2D(), C:bool=NOT Boolean(A=@n12)) -> (ReturnValue))
   > Set Alpha_CrouchMoving2D(Alpha_CrouchMoving2D:double=1) -> (Output_Get:double)
   Set Alpha_CrouchMoving2D(execute=@n13.else, Alpha_CrouchMoving2D=0)
   n14 = Get bIsPlayingEmote()
   n15 = Get bIsCrouching()
   n16 = Get bIsSkyDiving()
   n17 = Branch(execute=@n2.then_8, Condition=AND Boolean(A=NOT Boolean(A=@n16), B=NOT Boolean(A=@n14), C=NOT Boolean(A=@n15)))
   > Set Alpha_Standing(Alpha_Standing:double=1) -> (Output_Get:double)
   Set Alpha_Standing(execute=@n17.else, Alpha_Standing=0)
   n18 = NOT Boolean(A=Get bIsFalling())
   n19 = Get bIsMoving2D()
   n20 = Branch(execute=@n2.then_10, Condition=OR Boolean(A=AND Boolean!(A=@n19, B=@n18) -> (ReturnValue), B=Get bIsJumping()))
   > Set Alpha_Moving2D/Jumping(Alpha_Moving2D/Jumping:double=1) -> (Output_Get:double)
   Delay(execute=@n20.else, Duration=0.6)
   > Set Alpha_Moving2D/Jumping(Alpha_Moving2D/Jumping=0)
   n21 = Event OnCharacterCustomizationLoaded()
   n22 = Get Array Item with EFort Custom Part Type(Array=@n21.CustomMeshParts)
   > Set Body_Mesh(Body_Mesh=@n22)
   Event BlueprintInitializeAnimation()
   > Set Body_Mesh(Body_Mesh=@n22)
   Event OnWeaponChanged()
   > n23 = Cast To FortPlayerAnimInstance(Object=GetAnimInstance(self=GetOwnerSkeletalMeshComponent()))
   > Set Selected_Pistol(Selected_Pistol:double=0) -> (Output_Get:double)
   > Set Selected_Pickaxe(Selected_Pickaxe:double=0) -> (Output_Get:double)
   > Set NotSelected_Pistol(NotSelected_Pistol:double=1) -> (Output_Get:double)
   > n24 = Switch on EFortWeaponCoreAnimation!(execute, Selection:EFortWeaponCoreAnimation Enum=Get WeaponCoreAnim(self=@n23)) -> (
    Melee:Exec
    Pistol:Exec
    Shotgun:Exec
    PaperBlueprint:Exec
    Rifle:Exec
    MeleeOneHand:Exec
    MachinePistol:Exec
    RocketLauncher:Exec
    GrenadeLauncher:Exec
    GoingCommando:Exec
    AssaultRifle:Exec
    TacticalShotgun:Exec
    SniperRifle:Exec
    TrapPlacement:Exec
    ShoulderLauncher:Exec
    AbilityDecoTool:Exec
    Crossbow:Exec
    C4:Exec
    RemoteControl:Exec
    DualWield:Exec
    AR_BullPup:Exec
    AR_ForwardGrip:Exec
    MedPackPaddles:Exec
    SMG_P90:Exec
    AR_DrumGun:Exec
    Consumable_Small:Exec
    Consumable_Large:Exec
    Balloon:Exec
    MountedTurret:Exec
    CreativeTool:Exec
    ExplosiveBow:Exec
    AshtonIndigo:Exec
    AshtonChicago:Exec
    MeleeDualWield:Exec
    Unarmed:Exec
   )
   Set Selected_Pickaxe(execute=[@n24.Melee, @n24.MeleeDualWield], Selected_Pickaxe=1)
   Set Selected_Pistol(execute=@n24.Pistol, Selected_Pistol=1)
   > Set NotSelected_Pistol(NotSelected_Pistol=0)
   Event OnParachuteOpened()
   > Switch on EGliderType(Selection=Get PlayerGliderType() -> (PlayerGliderType=HangGlider))
   > Set HangGlidingAlpha(HangGlidingAlpha:double=1) -> (Output_Get:double)
   Event OnParachuteClosed()
   > Set HangGlidingAlpha(HangGlidingAlpha=0)
   n25 = Sequence!(execute) -> (then_0, then_1, then_2)
   n26 = GetOwnerSkeletalMeshComponent()
   n27 = BreakRotator(InRot=BreakTransform(InTransform=GetDeltaTransformFromRefPose(self=@n26, BoneName=neck_02, BaseName=spine_05)).Rotation)
   > n28 = Branch(Condition=AND Boolean(A="float < float"(A=@n27.Yaw, B=-33), B=Get bIsCrouching()))
   > Set Alpha_HeadRotateUpCrouchExtrem(Alpha_HeadRotateUpCrouchExtrem:double=1) -> (Output_Get:double)
   Set Alpha_HeadRotateUpCrouchExtrem(execute=@n28.else, Alpha_HeadRotateUpCrouchExtrem=0)
   n29 = Get bIsSkydiveDiveMode()
   n30 = Get bIsSkydiveFloating()
   n31 = GetOwnerSkeletalMeshComponent()
   n32 = BreakRotator(InRot=BreakTransform(InTransform=GetDeltaTransformFromRefPose(self=@n31, BoneName=head, BaseName=spine_01)).Rotation)
   n33 = GetOwnerSkeletalMeshComponent()
   n34 = BreakRotator(InRot=BreakTransform(InTransform=GetDeltaTransformFromRefPose(self=@n33, BoneName=neck_02, BaseName=spine_05)).Rotation)
   n35 = Branch(
    execute=@n25.then_1
    Condition=AND Boolean!(A="float > float"(A=@n34.Yaw, B=-3), B="float > float"(A=@n32.Yaw, B=-48), C:bool=OR Boolean(A=@n29, B=@n30)) -> (ReturnValue)
   )
   > Set Alpha_HeadRotateExtremFloating(Alpha_HeadRotateExtremFloating:double=1) -> (Output_Get:double)
   Set Alpha_HeadRotateExtremFloating(execute=@n35.else, Alpha_HeadRotateExtremFloating=0)
   n36 = Sequence!(execute=@n25.then_2) -> (then_0, then_1)
   n37 = GetOwnerSkeletalMeshComponent()
   n38 = BreakVector(InVec=BreakTransform(InTransform=GetDeltaTransformFromRefPose(self=@n37, BoneName=pelvis, BaseName=root)))
   n39 = Get bIsSkydiveFloating()
   > n40 = Branch(Condition=AND Boolean!(A="float < float"(A=@n38.Z, B=-20), B=@n39) -> (ReturnValue))
   > Set Alpha_RightFloating(Alpha_RightFloating:double=1) -> (Output_Get:double)
   Set Alpha_RightFloating(execute=@n40.else, Alpha_RightFloating=0)
   n41 = Branch(execute=@n36.then_1, Condition=AND Boolean(A=@n39, B="float > float"(A=@n38.Z, B=20)))
   > Set Alpha_LeftFloating(Alpha_LeftFloating:double=1) -> (Output_Get:double)
   Set Alpha_LeftFloating(execute=@n41.else, Alpha_LeftFloating=0)
  ]}
]
variables [
 Alpha_Crouching {type=number}
 Alpha_CrouchMoving2D {type=number}
 Alpha_CrouchMoving2DTargeting {type=number}
 Alpha_Floating {type=number}
 Alpha_HeadRotateExtremFloating {type=number}
 Alpha_HeadRotateUpCrouchExtrem {type=number}
 Alpha_LeftFloating {type=number}
 Alpha_Moving2D/Jumping {type=number}
 Alpha_NotTargeting {type=number}
 Alpha_RightFloating {type=number}
 Alpha_Standing {type=number}
 Alpha_Tactical_Sprinting {type=number}
 Alpha_Targeting {type=number}
 Body_Mesh {type=SkeletalMeshComponent}
 Delta_TimeX {type=number}
 Gravity Override Params {
  type=GravityOverrideParamsStruct
  default {skydive_FloatingDown {gravityOverride=<0 -1000 800>}, skydive_Parachuting {gravityOverride=<0 -1000 -500>}}
 }
 GravityOverride {type=Vector}
 HangGlidingAlpha {type=number}
 NotSelected_Pistol {type=number}
 PlayingCelebration {type=number}
 RBAN Params {
  type=RBANParamsStruct
  default {
   dBNO {simSpaceSettings {
     maxAngularAcceleration=1
     maxAngularVelocity=3
     maxLinearAcceleration=1000
     maxLinearVelocity=500
     velocityScaleZ=1
     externalLinearDragV=<1 1 1>
    }}
   dBNO_Carried {simSpaceSettings {
     maxAngularAcceleration=1
     maxAngularVelocity=3
     maxLinearAcceleration=1000
     maxLinearVelocity=500
     velocityScaleZ=1
     externalLinearDragV=<1 1 1>
    }}
   emoteOrMelee {simSpaceSettings {
     maxAngularAcceleration=1
     maxAngularVelocity=3
     maxLinearAcceleration=1000
     maxLinearVelocity=500
     velocityScaleZ=1
     externalLinearDragV=<1 1 1>
    }}
   front_End {simSpaceSettings {
     maxAngularAcceleration=1
     maxAngularVelocity=3
     maxLinearAcceleration=1000
     maxLinearVelocity=500
     velocityScaleZ=1
     externalLinearDragV=<1 1 1>
    }}
   onGround_Crouch_Moving {simSpaceSettings {
     maxAngularAcceleration=1
     maxAngularVelocity=3
     maxLinearAcceleration=1000
     maxLinearVelocity=500
     velocityScaleZ=1
     externalLinearDragV=<1 1 1>
    }}
   onGround_Falling {simSpaceSettings {
     maxAngularAcceleration=1
     maxAngularVelocity=3
     maxLinearAcceleration=1000
     maxLinearVelocity=500
     velocityScaleZ=1
     externalLinearDragV=<50 1 1>
    }}
   onGround_Moving {simSpaceSettings {
     maxAngularAcceleration=1
     maxAngularVelocity=3
     maxLinearAcceleration=1000
     maxLinearVelocity=500
     velocityScaleZ=1
     externalLinearDragV=<1 1 1>
    }}
   onGround_Standing {simSpaceSettings {
     maxAngularAcceleration=1
     maxAngularVelocity=3
     maxLinearAcceleration=1000
     maxLinearVelocity=500
     velocityScaleZ=1
     externalLinearDragV=<1 1 1>
    }}
   rail_Grinding {simSpaceSettings {
     maxAngularAcceleration=1
     maxAngularVelocity=3
     maxLinearAcceleration=1000
     maxLinearVelocity=500
     velocityScaleZ=1
     worldAlpha=1
     externalLinearDragV=<1 1 1>
    }}
   riding_Moving {simSpaceSettings {
     maxAngularAcceleration=1
     maxAngularVelocity=3
     maxLinearAcceleration=1000
     maxLinearVelocity=500
     velocityScaleZ=1
     externalLinearDragV=<1 1 1>
    }}
   riding_Moving_Targeting {simSpaceSettings {
     maxAngularAcceleration=1
     maxAngularVelocity=3
     maxLinearAcceleration=1000
     maxLinearVelocity=500
     velocityScaleZ=1
     externalLinearDragV=<1 1 1>
    }}
   riding_Standing {simSpaceSettings {
     maxAngularAcceleration=1
     maxAngularVelocity=3
     maxLinearAcceleration=1000
     maxLinearVelocity=500
     velocityScaleZ=1
     externalLinearDragV=<1 1 1>
    }}
   skydive_DivingDown {simSpaceSettings {
     maxAngularAcceleration=1
     maxAngularVelocity=3
     maxLinearAcceleration=1000
     maxLinearVelocity=500
     velocityScaleZ=1
     externalLinearDragV=<1 1 1>
    }}
   skydive_FloatingDown {
    componentLinearAccScale=<0 0.5 0>
    componentLinearVelScale=<0 0.5 0>
    simSpaceSettings {
     maxAngularAcceleration=1
     maxAngularVelocity=3
     maxLinearAcceleration=1000
     maxLinearVelocity=500
     velocityScaleZ=1
     externalLinearDragV=<1 1 1>
    }
   }
   skydive_GoingUp {simSpaceSettings {
     maxAngularAcceleration=1
     maxAngularVelocity=3
     maxLinearAcceleration=1000
     maxLinearVelocity=500
     velocityScaleZ=1
     externalLinearDragV=<1 1 1>
    }}
   skydive_Parachuting {
    componentLinearAccScale=<0 0.1 0.3>
    componentLinearVelScale=<0 0.1 0.3>
    simSpaceSettings {
     maxAngularAcceleration=1
     maxAngularVelocity=3
     maxLinearAcceleration=1000
     maxLinearVelocity=500
     velocityScaleZ=1
     externalLinearDragV=<1 1 1>
    }
   }
   skydive_Surfing {simSpaceSettings {
     maxAngularAcceleration=1
     maxAngularVelocity=3
     maxLinearAcceleration=1000
     maxLinearVelocity=500
     velocityScaleZ=1
     externalLinearDragV=<1 1 1>
    }}
   swimming {simSpaceSettings {
     maxAngularAcceleration=1
     maxAngularVelocity=3
     maxLinearAcceleration=1000
     maxLinearVelocity=500
     velocityScaleZ=1
     externalLinearDragV=<1 1 1>
    }}
   tactical_Sprint {simSpaceSettings {
     maxAngularAcceleration=1
     maxAngularVelocity=3
     maxLinearAcceleration=1000
     maxLinearVelocity=500
     velocityScaleZ=1
     externalLinearDragV=<1 1 1>
    }}
   ziplining {simSpaceSettings {
     maxAngularAcceleration=1
     maxAngularVelocity=3
     maxLinearAcceleration=1000
     maxLinearVelocity=500
     velocityScaleZ=1
     externalLinearDragV=<1 1 1>
    }}
  }
 }
 Selected_Pickaxe {type=number}
 Selected_Pistol {type=number}
 Sim Space Settings {
  type=SimSpaceSettings
  default {
   dampingAlpha=1
   maxAngularAcceleration=10000
   maxAngularVelocity=10000
   maxLinearAcceleration=10000
   maxLinearVelocity=10000
   velocityScaleZ=1
   worldAlpha=0
   externalLinearDragV=<0 0 0>
  }
 }
]
